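# Physics asset for the FE coat mesh of the male medium Relish Alt player body, defining secondary-motion bodies for the jacket.
# Capsules cover spine, pelvis and legs, plus zipper bones linked freely to the thighs.
# Back, side, front-side and front bone chains hang from jacket_base on spine_03, using boxes for lower front panels and capsules elsewhere.
# Chain links mostly have tight twist limits, with shapes and limits tuned asymmetrically left to right.
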PhysicsAsset /BRCosmetics/Characters/Player/Male/Medium/Bodies/M_MED_Relish_Alt/Meshes/M_MED_Relish_Alt_Coat_FE_Physics
bodies {
 calf_l [{shapeType=Capsule, length=41.472, radius=8.5, center=<-21.253 -2.7132 0.0004>, rotation=<rot 90 0 -178.9129>}]
 calf_r [{shapeType=Capsule, length=41.472, radius=8.5, center=<21.2646 0.6961 0.0003>, rotation=<rot 90 0 -178.9129>}]
 dyn_zipper_l_MSTR [{
   shapeType=Capsule
   length=25.6391
   radius=4.8708
   shapeName=dyn_zipper_l_MSTR_capsule
   center=<1.0643 -0.0001 -3.972>
   rotation=<rot -64.999 0 0>
  }]
 dyn_zipper_r_MSTR [{
   shapeType=Capsule
   length=25.5669
   radius=4.8708
   shapeName=dyn_zipper_r_MSTR_capsule
   center=<-1.1873 -0.0001 4.4308>
   rotation=<rot -74.9993 0 0>
  }]
 jacket_back_01_l [{shapeType=Capsule, length=13, radius=3}]
 jacket_back_01_r [{shapeType=Capsule, length=13, radius=3}]
 jacket_back_02_l [{shapeType=Capsule, length=10.8252, radius=3.4248, center=<-4.1139 -0.6153 -0.0001>}]
 jacket_back_02_r [{shapeType=Capsule, length=10.8252, radius=3.4248, center=<-4.1139 -0.6153 -0.0001>}]
 jacket_back_03_l [{shapeType=Capsule, length=16.3211, radius=2.6789, center=<-4.1139 -1.4464 0>}]
 jacket_back_03_r [{shapeType=Capsule, length=16.3211, radius=2.6789, center=<-4.1139 -1.4464 0>}]
 jacket_back_04_l [{shapeType=Capsule, length=15.6755, radius=2.8245, center=<-2.5637 -1.3176 0>}]
 jacket_back_04_r [{shapeType=Capsule, length=15.6755, radius=2.8245, center=<-3.2896 -1.3177 0>}]
 jacket_base [{shapeType=Capsule, length=0.505, radius=0.505, rotation=<rot 0 -90 0>}]
 jacket_frtSide_01_l [{shapeType=Capsule, length=15, radius=3, shapeName=jacket_frtSide_01_l_capsule, center=<2.0309 0.9622 0.3201>, rotation=<rot 0 0 -18.4003>}]
 jacket_frtSide_01_r [{shapeType=Capsule, length=15, radius=3.5, center=<-3.8329 0 0>, rotation=<rot 0 0 -20.0005>}]
 jacket_frtSide_02_l [{
   shapeType=Box
   extentX=20.5264
   extentY=4.6307
   extentZ=32.6742
   shapeName=jacket_frtSide_02_l_box
   center=<2.8334 1.5273 -0.4307>
   rotation=<rot 0 0 -15.0009>
  }]
 jacket_frtSide_02_r [{
   shapeType=Box
   extentX=15.2783
   extentY=6.9329
   extentZ=31.1901
   shapeName=jacket_frtSide_02_r_box
   center=<-4.0273 -2.6648 -0.9699>
   rotation=<rot 0 0 -20.0007>
  }]
 jacket_frtSide_03_l [{
   shapeType=Box
   extentX=19.8452
   extentY=5.2674
   extentZ=28.7744
   shapeName=jacket_frtSide_03_l_box
   center=<2.669 1.8606 -0.5118>
   rotation=<rot 0 0 -15.0004>
  }]
 jacket_frtSide_03_r [{
   shapeType=Box
   extentX=12.8475
   extentY=5.5994
   extentZ=30.3157
   shapeName=jacket_frtSide_03_r_box
   center=<-4.0273 -1.7853 -0.6498>
   rotation=<rot 0 0 -20.0001>
  }]
 jacket_frtSide_04_l [{
   shapeType=Box
   extentX=19.6569
   extentY=4.1799
   extentZ=30.1438
   shapeName=jacket_frtSide_04_l_box
   center=<4.5546 0.5919 0.6183>
   rotation=<rot 0 0 -15.0004>
  }]
 jacket_frtSide_04_r [{
   shapeType=Box
   extentX=12.235
   extentY=5.1508
   extentZ=27.5967
   shapeName=jacket_frtSide_04_r_box
   center=<-4.0273 -0.9638 -0.3508>
   rotation=<rot 0 0 -20.0001>
  }]
 jacket_frt_01_l [{shapeType=Capsule, length=5.75, radius=4.5, rotation=<rot 90 12.9074 7.9059>}]
 jacket_frt_01_r [{shapeType=Capsule, length=5.75, radius=4.5, rotation=<rot 90 -18.9704 156.0289>}]
 jacket_frt_02_l [{
   shapeType=Box
   extentX=17.4179
   extentY=17.3388
   extentZ=3.6991
   shapeName=jacket_frt_02_l_box
   center=<-3.8436 -0.2627 -0.4549>
   rotation=<rot 0 0 29.9994>
  }]
 jacket_frt_02_r [{
   shapeType=Box
   extentX=11.9686
   extentY=23.808
   extentZ=5.0882
   shapeName=jacket_frt_02_r_box
   center=<0 -0.2498 0.5357>
   rotation=<rot 0 0 -24.9991>
  }]
 jacket_frt_03_l [{
   shapeType=Box
   extentX=16.6131
   extentY=15.8233
   extentZ=4.0639
   shapeName=jacket_frt_03_l_box
   center=<-3.8436 -0.2627 -0.4549>
   rotation=<rot 0 0 30.0012>
  }]
 jacket_frt_03_r [{
   shapeType=Box
   extentX=14.0235
   extentY=23.5861
   extentZ=5.0402
   shapeName=jacket_frt_03_r_box
   center=<-1.4135 -0.2498 0.5358>
   rotation=<rot 0 0 -24.999>
  }]
 jacket_frt_04_l [{
   shapeType=Box
   extentX=16.0356
   extentY=15.9133
   extentZ=4.7551
   shapeName=jacket_frt_04_l_box
   center=<-3.8436 -0.2627 -0.4549>
   rotation=<rot 0 0 30.0006>
  }]
 jacket_frt_04_r [{
   shapeType=Box
   extentX=14.8346
   extentY=24.0036
   extentZ=6.2457
   shapeName=jacket_frt_04_r_box
   center=<-5.4238 -0.2498 0.5357>
   rotation=<rot 0 0 -24.9994>
  }]
 jacket_frt_05_l [{
   shapeType=Box
   extentX=25.5157
   extentY=24.3684
   extentZ=3.4386
   shapeName=jacket_frt_05_l_box
   center=<-3.8436 -0.2627 -0.4549>
   rotation=<rot 0 0 30.0001>
  }]
 jacket_frt_05_r [{
   shapeType=Box
   extentX=17.1909
   extentY=21.2857
   extentZ=4.8052
   shapeName=jacket_frt_05_r_box
   center=<-5.5118 -0.2498 0.5357>
   rotation=<rot 0 0 -24.999>
  }]
 jacket_side_01_l [{shapeType=Capsule, length=15, radius=3}]
 jacket_side_01_r [{shapeType=Capsule, length=15, radius=3}]
 jacket_side_02_l [{shapeType=Capsule, length=14.5752, radius=5.4248, center=<-4.7204 0 -0.0001>}]
 jacket_side_02_r [{shapeType=Capsule, length=17.5798, radius=3.4202, shapeName=jacket_side_02_r_capsule, center=<-2.2834 0 -0.1998>, rotation=<rot 5.0003 0 0>}]
 jacket_side_03_l [{shapeType=Capsule, length=14.5752, radius=5.4248, center=<-4.7204 0 -0.0001>}]
 jacket_side_03_r [{shapeType=Capsule, length=23.0798, radius=3.4202, shapeName=jacket_side_03_r_capsule, center=<-4.9303 -0.0001 0>, rotation=<rot 5 0 0>}]
 jacket_side_04_l [{shapeType=Capsule, length=30.7988, radius=3.4512, center=<-4.7204 0 -0.0001>}]
 jacket_side_04_r [{shapeType=Capsule, length=22.0561, radius=4.1938, shapeName=jacket_side_04_r_capsule, center=<-4.9303 -0.0001 0>}]
 pelvis [
  {shapeType=Capsule, length=9.75, radius=12, center=<-4.3449 -1.4432 0.0625>}
  {shapeType=Capsule, length=15, radius=3, shapeName=pelvis_capsule, center=<0.7246 0.0712 13.1534>, rotation=<rot 86.1367 0.0001 1.7947>}
 ]
 spine_03 [{shapeType=Capsule, length=21.75, radius=13.25, rotation=<rot 90 0 0>}]
 thigh_l [{shapeType=Capsule, length=37.7136, radius=9.3992, center=<-17.9638 2.2986 0.0495>, rotation=<rot 88.5882 -153.5743 24.357>}]
 thigh_r [{shapeType=Capsule, length=35.0831, radius=10.2436, center=<17.3153 -2.2317 -0.4202>, rotation=<rot 86.2405 -6.8173 171.0894>}]
}
constraints [
 {bone1Name=spine_03, bone2Name=pelvis, swing1LimitDegrees=45, swing2LimitDegrees=45, twistLimitDegrees=45}
 {bone1Name=thigh_r, bone2Name=pelvis, swing1LimitDegrees=45, swing2LimitDegrees=45, twistLimitDegrees=45}
 {bone1Name=calf_r, bone2Name=thigh_r, swing1LimitDegrees=45, swing2LimitDegrees=45, twistLimitDegrees=45}
 {bone1Name=jacket_base, bone2Name=spine_03, swing1LimitDegrees=45, swing2LimitDegrees=45, twistLimitDegrees=45}
 {bone1Name=jacket_back_01_r, bone2Name=jacket_base, swing1LimitDegrees=35, swing2LimitDegrees=15, twistLimitDegrees=1}
 {bone1Name=jacket_back_02_r, bone2Name=jacket_back_01_r, swing1LimitDegrees=35, swing2LimitDegrees=15, twistLimitDegrees=1}
 {bone1Name=jacket_back_03_r, bone2Name=jacket_back_02_r, swing1LimitDegrees=35, swing2LimitDegrees=15, twistLimitDegrees=1}
 {bone1Name=jacket_back_04_r, bone2Name=jacket_back_03_r, swing1LimitDegrees=35, swing2LimitDegrees=15, twistLimitDegrees=1}
 {bone1Name=jacket_back_01_l, bone2Name=jacket_base, swing1LimitDegrees=45, swing2LimitDegrees=45, twistLimitDegrees=1}
 {bone1Name=jacket_back_02_l, bone2Name=jacket_back_01_l, swing1LimitDegrees=35, swing2LimitDegrees=15, twistLimitDegrees=1}
 {bone1Name=jacket_back_03_l, bone2Name=jacket_back_02_l, swing1LimitDegrees=35, swing2LimitDegrees=15, twistLimitDegrees=1}
 {bone1Name=jacket_back_04_l, bone2Name=jacket_back_03_l, swing1LimitDegrees=35, swing2LimitDegrees=15, twistLimitDegrees=1}
 {bone1Name=jacket_side_01_l, bone2Name=jacket_base, swing1LimitDegrees=15, swing2LimitDegrees=30, twistLimitDegrees=5}
 {bone1Name=jacket_side_02_l, bone2Name=jacket_side_01_l, swing1LimitDegrees=35, swing2LimitDegrees=35, twistLimitDegrees=1}
 {bone1Name=jacket_side_03_l, bone2Name=jacket_side_02_l, swing1LimitDegrees=35, swing2LimitDegrees=35, twistLimitDegrees=1}
 {bone1Name=jacket_side_04_l, bone2Name=jacket_side_03_l, swing1LimitDegrees=35, swing2LimitDegrees=35, twistLimitDegrees=1}
 {bone1Name=jacket_frt_01_l, bone2Name=jacket_base, swing1LimitDegrees=45, swing2LimitDegrees=45, twistLimitDegrees=2}
 {bone1Name=jacket_frt_02_l, bone2Name=jacket_frt_01_l, swing1LimitDegrees=15, swing2LimitDegrees=35, twistLimitDegrees=1}
 {bone1Name=jacket_frt_03_l, bone2Name=jacket_frt_02_l, swing1LimitDegrees=15, swing2LimitDegrees=20, twistLimitDegrees=1}
 {bone1Name=jacket_frt_04_l, bone2Name=jacket_frt_03_l, swing1LimitDegrees=15, swing2LimitDegrees=20, twistLimitDegrees=1}
 {
  bone1Name=calf_l
  bone2Name=thigh_l
  swing1LimitDegrees=45
  swing1Motion=Free
  swing2LimitDegrees=45
  swing2Motion=Free
  twistLimitDegrees=45
  twistMotion=Free
 }
 {bone1Name=thigh_l, bone2Name=pelvis, swing1LimitDegrees=45, swing2LimitDegrees=45, twistLimitDegrees=45}
 {
  bone1Name=jacket_frtSide_01_r
  bone2Name=jacket_base
  swing1LimitDegrees=80
  swing1Motion=Free
  swing2LimitDegrees=60
  swing2Motion=Free
  twistLimitDegrees=8
  twistMotion=Free
 }
 {bone1Name=jacket_frtSide_02_r, bone2Name=jacket_frtSide_01_r, swing1LimitDegrees=35, swing2LimitDegrees=15, twistLimitDegrees=1}
 {bone1Name=jacket_frtSide_03_r, bone2Name=jacket_frtSide_02_r, swing1LimitDegrees=35, swing2LimitDegrees=15, twistLimitDegrees=5}
 {bone1Name=jacket_frtSide_04_r, bone2Name=jacket_frtSide_03_r, swing1LimitDegrees=35, swing2LimitDegrees=15, twistLimitDegrees=1}
 {bone1Name=jacket_frt_01_r, bone2Name=jacket_base, swing1LimitDegrees=45, swing2LimitDegrees=45, twistLimitDegrees=2}
 {bone1Name=jacket_frt_02_r, bone2Name=jacket_frt_01_r, swing1LimitDegrees=15, swing2LimitDegrees=35, twistLimitDegrees=1}
 {bone1Name=jacket_frt_03_r, bone2Name=jacket_frt_02_r, swing1LimitDegrees=15, swing2LimitDegrees=20, twistLimitDegrees=1}
 {bone1Name=jacket_frt_04_r, bone2Name=jacket_frt_03_r, swing1LimitDegrees=15, swing2LimitDegrees=20, twistLimitDegrees=1}
 {bone1Name=jacket_side_01_r, bone2Name=jacket_base, swing1LimitDegrees=15, swing2LimitDegrees=30, twistLimitDegrees=5}
 {bone1Name=jacket_frt_05_r, bone2Name=jacket_frt_04_r, swing1LimitDegrees=15, swing2LimitDegrees=35, twistLimitDegrees=1}
 {bone1Name=jacket_frt_05_l, bone2Name=jacket_frt_04_l, swing1LimitDegrees=15, swing2LimitDegrees=35, twistLimitDegrees=1}
 {bone1Name=jacket_frtSide_01_l, bone2Name=jacket_base, swing1LimitDegrees=2, swing2LimitDegrees=5, twistLimitDegrees=1}
 {bone1Name=jacket_frtSide_02_l, bone2Name=jacket_frtSide_01_l, swing1LimitDegrees=35, swing2LimitDegrees=35, twistLimitDegrees=5}
 {bone1Name=jacket_frtSide_03_l, bone2Name=jacket_frtSide_02_l, swing1LimitDegrees=35, swing2LimitDegrees=35, twistLimitDegrees=5}
 {bone1Name=jacket_frtSide_04_l, bone2Name=jacket_frtSide_03_l, swing1LimitDegrees=35, swing2LimitDegrees=35, twistLimitDegrees=5}
 {bone1Name=jacket_side_02_r, bone2Name=jacket_side_01_r, swing1LimitDegrees=30, swing2LimitDegrees=15, twistLimitDegrees=0.1}
 {bone1Name=jacket_side_03_r, bone2Name=jacket_side_02_r, swing1LimitDegrees=30, swing2LimitDegrees=15, twistLimitDegrees=0.1}
 {bone1Name=jacket_side_04_r, bone2Name=jacket_side_03_r, swing1LimitDegrees=30, swing2LimitDegrees=15, twistLimitDegrees=0.1}
 {
  bone1Name=dyn_zipper_l_MSTR
  bone2Name=thigh_l
  swing1LimitDegrees=45
  swing1Motion=Free
  swing2LimitDegrees=45
  swing2Motion=Free
  twistLimitDegrees=45
  twistMotion=Free
 }
 {
  bone1Name=dyn_zipper_r_MSTR
  bone2Name=thigh_r
  swing1LimitDegrees=45
  swing1Motion=Free
  swing2LimitDegrees=45
  swing2Motion=Free
  twistLimitDegrees=45
  twistMotion=Free
 }
]
