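# Physics asset for the coat of the M_MED_Relish_Alt player cosmetic body, defining capsule bodies and joints for its jacket bones.
# Left and right back, side, front-side and front bone chains hang from jacket_base, which joins spine_03, alongside pelvis, thigh and calf capsules.
# Chain joints mostly have tight twist and asymmetrically tuned swing limits.
# The calf_l and jacket_frtSide_01_r joints are fully free angularly, ignoring their limit values.
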PhysicsAsset /BRCosmetics/Characters/Player/Male/Medium/Bodies/M_MED_Relish_Alt/Meshes/M_MED_Relish_Alt_Coat_Physics
bodies {
 calf_l [{shapeType=Capsule, length=41.472, radius=8.5, center=<-21.253 -2.7132 0.0004>, rotation=<rot 90 0 -178.9129>}]
 calf_r [{shapeType=Capsule, length=41.472, radius=8.5, center=<21.2646 0.6961 0.0003>, rotation=<rot 90 0 -178.9129>}]
 jacket_back_01_l [{shapeType=Capsule, length=13, radius=3}]
 jacket_back_01_r [{shapeType=Capsule, length=13, radius=3}]
 jacket_back_02_l [{shapeType=Capsule, length=11.25, radius=3, center=<-4.1139 1.3158 -0.0002>}]
 jacket_back_02_r [{shapeType=Capsule, length=11.25, radius=3, center=<-4.1139 1.3158 -0.0002>}]
 jacket_back_03_l [{shapeType=Capsule, length=15.5, radius=3.5, center=<-4.1139 -0.1707 0>}]
 jacket_back_03_r [{shapeType=Capsule, length=15.5, radius=3.5, center=<-4.1139 -0.1707 0>}]
 jacket_back_04_l [{shapeType=Capsule, length=15, radius=3.5, center=<-2.5637 0.0001 0>}]
 jacket_back_04_r [{shapeType=Capsule, length=15, radius=3.5, center=<-3.2896 -0.0001 0>}]
 jacket_base [{shapeType=Capsule, length=0.505, radius=0.505, rotation=<rot 0 -90 0>}]
 jacket_frtSide_01_l [{shapeType=Capsule, length=15, radius=3, shapeName=jacket_frtSide_01_l_capsule, center=<2.0309 0.9622 0.3201>, rotation=<rot 0 0 -18.4003>}]
 jacket_frtSide_01_r [{shapeType=Capsule, length=15, radius=3.5, center=<-3.8329 0 0>, rotation=<rot 0 0 -20.0005>}]
 jacket_frtSide_02_l [{
   shapeType=Capsule
   length=19.4121
   radius=3.3078
   shapeName=jacket_frtSide_02_l_capsule
   center=<2.0309 0.9622 0.3202>
   rotation=<rot 0 0 -18.4092>
  }]
 jacket_frtSide_02_r [{shapeType=Capsule, length=25, radius=3.25, center=<-6.4001 0 0.0001>, rotation=<rot 0 0 -30.0001>}]
 jacket_frtSide_03_l [{shapeType=Capsule, length=21.25, radius=3.5, shapeName=jacket_frtSide_03_l_capsule, center=<2.0309 0.9622 0.32>, rotation=<rot 0 0 -18.398>}]
 jacket_frtSide_03_r [{shapeType=Capsule, length=28.5, radius=3.25, center=<-3.3075 -0.2732 -0.1578>, rotation=<rot 0 0 -30>}]
 jacket_frtSide_04_l [{shapeType=Capsule, length=19.5, radius=4, shapeName=jacket_frtSide_04_l_capsule, center=<2.0309 0.9621 0.3203>, rotation=<rot 0 0 -18.4113>}]
 jacket_frtSide_04_r [{shapeType=Capsule, length=22.5, radius=3, center=<-3.3075 0 0>, rotation=<rot 0 0 -30>}]
 jacket_frt_01_l [{shapeType=Capsule, length=5.75, radius=4.5, rotation=<rot 90 12.9074 7.9059>}]
 jacket_frt_01_r [{shapeType=Capsule, length=5.75, radius=4.5, rotation=<rot 90 -18.9704 156.0289>}]
 jacket_frt_02_l [{
   shapeType=Capsule
   length=17.1051
   radius=2.7591
   shapeName=jacket_frt_02_l_capsule
   center=<-3.8134 0.2054 0.1257>
   rotation=<rot 0 0 -60.0015>
  }]
 jacket_frt_02_r [{shapeType=Capsule, length=10.25, radius=2.75, shapeName=jacket_frt_02_r_capsule, center=<-2.4579 0 0>, rotation=<rot 0 0 50.0017>}]
 jacket_frt_03_l [{
   shapeType=Capsule
   length=23.3597
   radius=3.1831
   shapeName=jacket_frt_03_l_capsule
   center=<-5.8846 0.5074 -1.482>
   rotation=<rot 0 0 -59.9995>
  }]
 jacket_frt_03_r [{shapeType=Capsule, length=18, radius=3.25, shapeName=jacket_frt_03_r_capsule, center=<-6.7137 0 0>, rotation=<rot 0 0 55.0007>}]
 jacket_frt_04_l [{
   shapeType=Capsule
   length=20.5814
   radius=4.25
   shapeName=jacket_frt_04_l_capsule
   center=<-7.7459 -0.9532 -1.9068>
   rotation=<rot 0 0 -56.5218>
  }]
 jacket_frt_04_r [{shapeType=Capsule, length=18, radius=3.9, shapeName=jacket_frt_04_r_capsule, center=<-6.6076 0 0>, rotation=<rot 0 0 49.9986>}]
 jacket_frt_05_l [{
   shapeType=Capsule
   length=20.6387
   radius=3.9623
   shapeName=jacket_frt_05_l_capsule
   center=<-5.1215 -0.4021 -0.8247>
   rotation=<rot 0 0 -34.9998>
  }]
 jacket_frt_05_r [{shapeType=Capsule, length=14, radius=4.65, shapeName=jacket_frt_05_r_capsule, center=<-2.4579 0 0>, rotation=<rot 0 0 49.9992>}]
 jacket_side_01_l [{shapeType=Capsule, length=15, radius=3}]
 jacket_side_01_r [{shapeType=Capsule, length=15, radius=3}]
 jacket_side_02_l [{shapeType=Capsule, length=15, radius=5}]
 jacket_side_02_r [{shapeType=Capsule, length=16, radius=5, shapeName=jacket_side_02_r_capsule, center=<-2.2834 0 -0.1998>, rotation=<rot 5.0003 0 0>}]
 jacket_side_03_l [{shapeType=Capsule, length=15, radius=5}]
 jacket_side_03_r [{shapeType=Capsule, length=21.5, radius=5, shapeName=jacket_side_03_r_capsule, center=<-4.9303 -0.0001 0>, rotation=<rot 5 0 0>}]
 jacket_side_04_l [{shapeType=Capsule, length=29.25, radius=5}]
 jacket_side_04_r [{shapeType=Capsule, length=21.25, radius=5, shapeName=jacket_side_04_r_capsule, center=<-4.9303 -0.0001 0>}]
 pelvis [
  {shapeType=Capsule, length=9.75, radius=12, center=<-4.3449 -1.4432 0.0625>}
  {shapeType=Capsule, length=15, radius=3, shapeName=pelvis_capsule, center=<0.7246 0.0712 13.1534>, rotation=<rot 86.1367 0.0001 1.7947>}
 ]
 spine_03 [{shapeType=Capsule, length=21.75, radius=13.25, rotation=<rot 90 0 0>}]
 thigh_l [{shapeType=Capsule, length=37.25, radius=9.25, center=<-17.8814 0.0212 0.0263>, rotation=<rot 88.5881 -153.5746 24.3569>}]
 thigh_r [{shapeType=Capsule, length=25.25, radius=9.25, center=<19.4189 -0.0436 0.7846>, rotation=<rot 88.5886 -153.5746 24.3573>}]
}
constraints [
 {bone1Name=spine_03, bone2Name=pelvis, swing1LimitDegrees=45, swing2LimitDegrees=45, twistLimitDegrees=45}
 {bone1Name=thigh_r, bone2Name=pelvis, swing1LimitDegrees=45, swing2LimitDegrees=45, twistLimitDegrees=45}
 {bone1Name=calf_r, bone2Name=thigh_r, swing1LimitDegrees=45, swing2LimitDegrees=45, twistLimitDegrees=45}
 {bone1Name=jacket_base, bone2Name=spine_03, swing1LimitDegrees=45, swing2LimitDegrees=45, twistLimitDegrees=45}
 {bone1Name=jacket_back_01_r, bone2Name=jacket_base, swing1LimitDegrees=35, swing2LimitDegrees=15, twistLimitDegrees=1}
 {bone1Name=jacket_back_02_r, bone2Name=jacket_back_01_r, swing1LimitDegrees=35, swing2LimitDegrees=15, twistLimitDegrees=1}
 {bone1Name=jacket_back_03_r, bone2Name=jacket_back_02_r, swing1LimitDegrees=35, swing2LimitDegrees=15, twistLimitDegrees=1}
 {bone1Name=jacket_back_04_r, bone2Name=jacket_back_03_r, swing1LimitDegrees=35, swing2LimitDegrees=15, twistLimitDegrees=1}
 {bone1Name=jacket_back_01_l, bone2Name=jacket_base, swing1LimitDegrees=45, swing2LimitDegrees=45, twistLimitDegrees=1}
 {bone1Name=jacket_back_02_l, bone2Name=jacket_back_01_l, swing1LimitDegrees=35, swing2LimitDegrees=15, twistLimitDegrees=1}
 {bone1Name=jacket_back_03_l, bone2Name=jacket_back_02_l, swing1LimitDegrees=35, swing2LimitDegrees=15, twistLimitDegrees=1}
 {bone1Name=jacket_back_04_l, bone2Name=jacket_back_03_l, swing1LimitDegrees=35, swing2LimitDegrees=15, twistLimitDegrees=1}
 {bone1Name=jacket_side_01_l, bone2Name=jacket_base, swing1LimitDegrees=15, swing2LimitDegrees=30, twistLimitDegrees=5}
 {bone1Name=jacket_side_02_l, bone2Name=jacket_side_01_l, swing1LimitDegrees=15, swing2LimitDegrees=35, twistLimitDegrees=1}
 {bone1Name=jacket_side_03_l, bone2Name=jacket_side_02_l, swing1LimitDegrees=15, swing2LimitDegrees=35, twistLimitDegrees=1}
 {bone1Name=jacket_side_04_l, bone2Name=jacket_side_03_l, swing1LimitDegrees=15, swing2LimitDegrees=35, twistLimitDegrees=1}
 {bone1Name=jacket_frt_01_l, bone2Name=jacket_base, swing1LimitDegrees=45, swing2LimitDegrees=45, twistLimitDegrees=2}
 {bone1Name=jacket_frt_02_l, bone2Name=jacket_frt_01_l, swing1LimitDegrees=15, swing2LimitDegrees=35, twistLimitDegrees=1}
 {bone1Name=jacket_frt_03_l, bone2Name=jacket_frt_02_l, swing1LimitDegrees=15, swing2LimitDegrees=20, twistLimitDegrees=1}
 {bone1Name=jacket_frt_04_l, bone2Name=jacket_frt_03_l, swing1LimitDegrees=15, swing2LimitDegrees=20, twistLimitDegrees=1}
 {
  bone1Name=calf_l
  bone2Name=thigh_l
  swing1LimitDegrees=45
  swing1Motion=Free
  swing2LimitDegrees=45
  swing2Motion=Free
  twistLimitDegrees=45
  twistMotion=Free
 }
 {bone1Name=thigh_l, bone2Name=pelvis, swing1LimitDegrees=45, swing2LimitDegrees=45, twistLimitDegrees=45}
 {
  bone1Name=jacket_frtSide_01_r
  bone2Name=jacket_base
  swing1LimitDegrees=80
  swing1Motion=Free
  swing2LimitDegrees=60
  swing2Motion=Free
  twistLimitDegrees=8
  twistMotion=Free
 }
 {bone1Name=jacket_frtSide_02_r, bone2Name=jacket_frtSide_01_r, swing1LimitDegrees=35, swing2LimitDegrees=15, twistLimitDegrees=1}
 {bone1Name=jacket_frtSide_03_r, bone2Name=jacket_frtSide_02_r, swing1LimitDegrees=35, swing2LimitDegrees=15, twistLimitDegrees=5}
 {bone1Name=jacket_frtSide_04_r, bone2Name=jacket_frtSide_03_r, swing1LimitDegrees=35, swing2LimitDegrees=15, twistLimitDegrees=1}
 {bone1Name=jacket_frt_01_r, bone2Name=jacket_base, swing1LimitDegrees=45, swing2LimitDegrees=45, twistLimitDegrees=2}
 {bone1Name=jacket_frt_02_r, bone2Name=jacket_frt_01_r, swing1LimitDegrees=15, swing2LimitDegrees=35, twistLimitDegrees=1}
 {bone1Name=jacket_frt_03_r, bone2Name=jacket_frt_02_r, swing1LimitDegrees=15, swing2LimitDegrees=20, twistLimitDegrees=1}
 {bone1Name=jacket_frt_04_r, bone2Name=jacket_frt_03_r, swing1LimitDegrees=15, swing2LimitDegrees=20, twistLimitDegrees=1}
 {bone1Name=jacket_side_01_r, bone2Name=jacket_base, swing1LimitDegrees=15, swing2LimitDegrees=30, twistLimitDegrees=5}
 {bone1Name=jacket_frt_05_r, bone2Name=jacket_frt_04_r, swing1LimitDegrees=15, swing2LimitDegrees=35, twistLimitDegrees=1}
 {bone1Name=jacket_frt_05_l, bone2Name=jacket_frt_04_l, swing1LimitDegrees=15, swing2LimitDegrees=35, twistLimitDegrees=1}
 {bone1Name=jacket_frtSide_01_l, bone2Name=jacket_base, swing1LimitDegrees=2, swing2LimitDegrees=5, twistLimitDegrees=1}
 {bone1Name=jacket_frtSide_02_l, bone2Name=jacket_frtSide_01_l, swing1LimitDegrees=35, swing2LimitDegrees=35, twistLimitDegrees=5}
 {bone1Name=jacket_frtSide_03_l, bone2Name=jacket_frtSide_02_l, swing1LimitDegrees=35, swing2LimitDegrees=35, twistLimitDegrees=5}
 {bone1Name=jacket_frtSide_04_l, bone2Name=jacket_frtSide_03_l, swing1LimitDegrees=35, swing2LimitDegrees=35, twistLimitDegrees=5}
 {bone1Name=jacket_side_02_r, bone2Name=jacket_side_01_r, swing1LimitDegrees=30, swing2LimitDegrees=15, twistLimitDegrees=0.1}
 {bone1Name=jacket_side_03_r, bone2Name=jacket_side_02_r, swing1LimitDegrees=30, swing2LimitDegrees=15, twistLimitDegrees=0.1}
 {bone1Name=jacket_side_04_r, bone2Name=jacket_side_03_r, swing1LimitDegrees=30, swing2LimitDegrees=15, twistLimitDegrees=0.1}
]
