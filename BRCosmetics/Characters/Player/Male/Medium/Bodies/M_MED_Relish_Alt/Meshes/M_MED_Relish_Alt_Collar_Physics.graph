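# Physics asset for the M_MED_Relish_Alt male character's collar, covering dynamic collar and lapel bones plus the surrounding upper body.
# Every body is a single capsule, on the spine, neck, clavicles, upper arms and four dyn_ bones.
# Collar bones constrain to spine_05 and lapels to spine_04 with tight, asymmetric swing and twist limits, while clavicle and neck joints rotate freely.
# The left collar bone is spelled dyn_collal_l, unlike dyn_collar_r.
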PhysicsAsset /BRCosmetics/Characters/Player/Male/Medium/Bodies/M_MED_Relish_Alt/Meshes/M_MED_Relish_Alt_Collar_Physics
bodies {
 clavicle_l [{shapeType=Capsule, length=12.25, radius=3, shapeName=clavicle_l_capsule, center=<10.8592 -1.7911 5.251>, rotation=<rot 90 0 0>}]
 clavicle_r [{shapeType=Capsule, length=11, radius=5.5, shapeName=clavicle_r_capsule, center=<-10.0363 1.6807 -6.0724>, rotation=<rot 74.9981 0 -10.0012>}]
 dyn_collal_l [{shapeType=Capsule, length=14, radius=3, shapeName=dyn_collal_l_capsule, center=<-6.0311 -0.9449 4.9916>, rotation=<rot 0 91.1239 -63.3994>}]
 dyn_collar_r [{shapeType=Capsule, length=10.5, radius=4, shapeName=dyn_collar_r_capsule, center=<-4.0607 0.0188 -4.6982>, rotation=<rot 90 0 0>}]
 dyn_lapel_l [{shapeType=Capsule, length=11.5, radius=2.5, shapeName=dyn_lapel_l_capsule, center=<2.3464 -0.5913 4.8334>, rotation=<rot 0 -84.063 -43.2954>}]
 dyn_lapel_r [{shapeType=Capsule, length=12, radius=2.5, shapeName=dyn_lapel_r_capsule, center=<2.4666 0.0084 -5.331>, rotation=<rot 0 93.0455 -54.1628>}]
 neck_02 [{shapeType=Capsule, length=1.25, radius=8, shapeName=neck_02_capsule, rotation=<rot 0 -90 0>}]
 spine_04 [{shapeType=Capsule, length=19.7972, radius=12.6403, center=<-0.391 0.0111 -0.0893>, rotation=<rot 0.2616 -90 -0.2715>}]
 spine_05 [{shapeType=Capsule, length=20.4703, radius=11.3309, center=<1.7221 -0.7463 -0.1453>, rotation=<rot 2.5041 -90 -1.0809>}]
 upperarm_l [{shapeType=Capsule, length=18.9015, radius=7.5492, center=<13.578 -0.0295 -1.0211>, rotation=<rot 0 -89.2971 -85.2172>}]
 upperarm_r [{shapeType=Capsule, length=23.1524, radius=8.2358, center=<-10.486 -0.0299 -0.216>, rotation=<rot -0.2089 -87.3226 -87.6107>}]
}
constraints [
 {bone1Name=spine_05, bone2Name=spine_04, swing1LimitDegrees=45, swing2LimitDegrees=45, twistLimitDegrees=45}
 {bone1Name=upperarm_l, bone2Name=clavicle_l, swing1LimitDegrees=45, swing2LimitDegrees=45, twistLimitDegrees=45}
 {bone1Name=upperarm_r, bone2Name=clavicle_r, swing1LimitDegrees=45, swing2LimitDegrees=45, twistLimitDegrees=45}
 {bone1Name=dyn_collal_l, bone2Name=spine_05, swing1LimitDegrees=15, swing2LimitDegrees=30, twistLimitDegrees=5}
 {bone1Name=dyn_collar_r, bone2Name=spine_05, swing1LimitDegrees=15, swing2LimitDegrees=30, twistLimitDegrees=5}
 {bone1Name=dyn_lapel_r, bone2Name=spine_04, swing1LimitDegrees=15, swing2LimitDegrees=30, twistLimitDegrees=5}
 {bone1Name=dyn_lapel_l, bone2Name=spine_04, swing1LimitDegrees=15, swing2LimitDegrees=30, twistLimitDegrees=5}
 {
  bone1Name=clavicle_l
  bone2Name=spine_05
  swing1LimitDegrees=45
  swing1Motion=Free
  swing2LimitDegrees=45
  swing2Motion=Free
  twistLimitDegrees=45
  twistMotion=Free
 }
 {
  bone1Name=clavicle_r
  bone2Name=spine_05
  swing1LimitDegrees=45
  swing1Motion=Free
  swing2LimitDegrees=45
  swing2Motion=Free
  twistLimitDegrees=45
  twistMotion=Free
 }
 {
  bone1Name=neck_02
  bone2Name=spine_05
  swing1LimitDegrees=45
  swing1Motion=Free
  swing2LimitDegrees=45
  swing2Motion=Free
  twistLimitDegrees=45
  twistMotion=Free
 }
]
