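# Crouch physics asset for the medium male Relish Alt player body, set up for jacket-bone secondary motion.
# Left and right back, side, front-side and front capsule chains hang from jacket_base on spine_03, mostly with tight twist limits and asymmetric tuning.
# Pelvis, chest and leg capsules provide body shapes for the jacket.
# A huge zero-length capsule on root, named floorCOL_sphere, acts as a floor collider beneath the character.
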PhysicsAsset /BRCosmetics/Characters/Player/Male/Medium/Bodies/M_MED_Relish_Alt/Meshes/M_MED_Relish_Alt_Crouch_Physics
bodies {
 calf_l [{shapeType=Capsule, length=41.472, radius=8.5, center=<-21.253 -2.7132 0.0004>, rotation=<rot 90 0 -178.9129>}]
 calf_r [{shapeType=Capsule, length=41.472, radius=8.5, center=<21.2646 0.6961 0.0003>, rotation=<rot 90 0 -178.9129>}]
 jacket_back_01_l [{shapeType=Capsule, length=13, radius=3}]
 jacket_back_01_r [{shapeType=Capsule, length=13, radius=3}]
 jacket_back_02_l [{shapeType=Capsule, length=11.5, radius=2.75, center=<-4.1139 1.3158 -0.0002>}]
 jacket_back_02_r [{shapeType=Capsule, length=11.5, radius=2.75, center=<-4.1139 1.3158 -0.0002>}]
 jacket_back_03_l [{shapeType=Capsule, length=15.5, radius=2, center=<-4.1139 -0.1707 0>}]
 jacket_back_03_r [{shapeType=Capsule, length=15.5, radius=2, center=<-4.1139 -0.1707 0>}]
 jacket_back_04_l [{shapeType=Capsule, length=15, radius=5, center=<-2.5637 0.0001 0>}]
 jacket_back_04_r [{shapeType=Capsule, length=15, radius=5, center=<-3.2896 -0.0001 0>}]
 jacket_base [{shapeType=Capsule, length=0.505, radius=0.505, rotation=<rot 0 -90 0>}]
 jacket_frtSide_01_l [{shapeType=Capsule, length=15, radius=3, shapeName=jacket_frtSide_01_l_capsule, center=<3.3339 0 0>, rotation=<rot 0 0 -10.0005>}]
 jacket_frtSide_01_r [{shapeType=Capsule, length=15, radius=4.75, center=<-4.4196 0 0>}]
 jacket_frtSide_02_l [{shapeType=Capsule, length=18.75, radius=4.25, shapeName=jacket_frtSide_02_l_capsule, center=<4.5434 0.4832 0>, rotation=<rot 0 -105.0006 0>}]
 jacket_frtSide_02_r [{shapeType=Capsule, length=13.5, radius=4.75, center=<-8.7575 -0.0002 -0.0001>}]
 jacket_frtSide_03_l [{
   shapeType=Capsule
   length=28.5
   radius=5.5
   shapeName=jacket_frtSide_03_l_capsule
   center=<3.7352 0.3144 0.3405>
   rotation=<rot -14.9998 -105.0012 0>
  }]
 jacket_frtSide_03_r [{shapeType=Capsule, length=15, radius=4.75, center=<-8.1428 -0.0001 0>}]
 jacket_frtSide_04_l [{
   shapeType=Capsule
   length=20.25
   radius=3
   shapeName=jacket_frtSide_04_l_capsule
   center=<6.3486 -0.0001 0>
   rotation=<rot -0.0001 -105.0012 -179.9991>
  }]
 jacket_frtSide_04_r [{shapeType=Capsule, length=17.25, radius=4.75, center=<-7.7995 -0.0002 -0.0003>}]
 jacket_frt_01_l [{shapeType=Capsule, length=5.75, radius=4.5, rotation=<rot 90 12.9074 7.9059>}]
 jacket_frt_01_r [{shapeType=Capsule, length=5.75, radius=4.5, rotation=<rot 90 -18.9704 156.0289>}]
 jacket_frt_02_l [{shapeType=Capsule, length=20, radius=3, shapeName=jacket_frt_02_l_capsule, rotation=<rot 60.001 94.9978 0>}]
 jacket_frt_02_r [{shapeType=Capsule, length=10.25, radius=4, shapeName=jacket_frt_02_r_capsule, center=<-5.0903 -0.0491 -1.9964>, rotation=<rot 0 0 45>}]
 jacket_frt_03_l [{shapeType=Capsule, length=20, radius=3, shapeName=jacket_frt_03_l_capsule, center=<-1.3843 0.4928 1.0591>, rotation=<rot 60.001 94.9978 0>}]
 jacket_frt_03_r [{shapeType=Capsule, length=12.25, radius=4.75, shapeName=jacket_frt_03_r_capsule, center=<-6.0796 -0.8981 -1.1089>, rotation=<rot 0 0 45>}]
 jacket_frt_04_l [{
   shapeType=Capsule
   length=17.3314
   radius=3
   shapeName=jacket_frt_04_l_capsule
   center=<-4.6001 0.3514 1.3005>
   rotation=<rot 60.001 94.9978 0>
  }]
 jacket_frt_04_r [{shapeType=Capsule, length=10.5, radius=4.75, shapeName=jacket_frt_04_r_capsule, center=<-7.6221 0.6137 -1.2031>, rotation=<rot 0 0 45>}]
 jacket_frt_05_l [{
   shapeType=Capsule
   length=20.6387
   radius=5.7123
   shapeName=jacket_frt_05_l_capsule
   center=<-4.6122 -0.4032 -0.0002>
   rotation=<rot 59.9979 94.9947 180>
  }]
 jacket_frt_05_r [{shapeType=Capsule, length=12, radius=4.75, shapeName=jacket_frt_05_r_capsule, center=<-4.2634 -0.4252 -0.8562>, rotation=<rot 0 0 45>}]
 jacket_side_01_l [{shapeType=Capsule, length=15, radius=3, rotation=<rot 0 0 -10>}]
 jacket_side_01_r [{shapeType=Capsule, length=15, radius=3, rotation=<rot 0 0 10>}]
 jacket_side_02_l [{shapeType=Capsule, length=20, radius=4}]
 jacket_side_02_r [{shapeType=Capsule, length=14.25, radius=5, shapeName=jacket_side_02_r_capsule, center=<-3.9017 -0.4993 2.383>}]
 jacket_side_03_l [{shapeType=Capsule, length=39, radius=5, center=<-3.702 0.0001 0>}]
 jacket_side_03_r [{shapeType=Capsule, length=14.25, radius=5, shapeName=jacket_side_03_r_capsule, center=<-8.4952 -0.5935 -0.4322>}]
 jacket_side_04_l [{shapeType=Capsule, length=26, radius=5, center=<-1.3986 0.0001 0>}]
 jacket_side_04_r [{shapeType=Capsule, length=20.5, radius=5, shapeName=jacket_side_04_r_capsule, center=<-8.1095 -0.7982 1.0618>}]
 pelvis [{shapeType=Capsule, length=7.25, radius=12.5, center=<-4.3449 -0.9936 0.0625>}]
 root [{shapeType=Capsule, length=0, radius=333, shapeName=floorCOL_sphere, center=<-19.6028 20.7744 -323.1712>, rotation=<rot 0 -90 0>}]
 spine_03 [{shapeType=Capsule, length=22.25, radius=12.75, rotation=<rot 90 0 0>}]
 thigh_l [{shapeType=Capsule, length=52, radius=9, center=<-19.2747 -0.0289 -0.005>, rotation=<rot 88.5868 -153.5741 24.3564>}]
 thigh_r [{shapeType=Capsule, length=42.25, radius=8.75, center=<19.4189 -0.0436 0.7846>, rotation=<rot 88.5886 -153.5746 24.3573>}]
 thigh_twist_01_l [{
   shapeType=Capsule
   length=14.6956
   radius=6.2082
   shapeName=thigh_twist_01_l_capsule
   center=<6.42 2.9543 -3.4625>
   rotation=<rot -27.4741 -90 -90.8135>
  }]
}
constraints [
 {bone1Name=spine_03, bone2Name=pelvis, swing1LimitDegrees=45, swing2LimitDegrees=45, twistLimitDegrees=45}
 {bone1Name=thigh_r, bone2Name=pelvis, swing1LimitDegrees=45, swing2LimitDegrees=45, twistLimitDegrees=45}
 {bone1Name=calf_r, bone2Name=thigh_r, swing1LimitDegrees=45, swing2LimitDegrees=45, twistLimitDegrees=45}
 {bone1Name=jacket_base, bone2Name=spine_03, swing1LimitDegrees=45, swing2LimitDegrees=45, twistLimitDegrees=45}
 {bone1Name=jacket_back_01_r, bone2Name=jacket_base, swing1LimitDegrees=45, swing2LimitDegrees=45, twistLimitDegrees=1}
 {bone1Name=jacket_back_02_r, bone2Name=jacket_back_01_r, swing1LimitDegrees=45, swing2LimitDegrees=45, twistLimitDegrees=1}
 {bone1Name=jacket_back_03_r, bone2Name=jacket_back_02_r, swing1LimitDegrees=45, swing2LimitDegrees=45, twistLimitDegrees=1}
 {bone1Name=jacket_back_04_r, bone2Name=jacket_back_03_r, swing1LimitDegrees=45, swing2LimitDegrees=45, twistLimitDegrees=1}
 {bone1Name=jacket_back_01_l, bone2Name=jacket_base, swing1LimitDegrees=45, swing2LimitDegrees=45, twistLimitDegrees=1}
 {bone1Name=jacket_back_02_l, bone2Name=jacket_back_01_l, swing1LimitDegrees=45, swing2LimitDegrees=45, twistLimitDegrees=1}
 {bone1Name=jacket_back_03_l, bone2Name=jacket_back_02_l, swing1LimitDegrees=45, swing2LimitDegrees=45, twistLimitDegrees=1}
 {bone1Name=jacket_back_04_l, bone2Name=jacket_back_03_l, swing1LimitDegrees=45, swing2LimitDegrees=45, twistLimitDegrees=1}
 {bone1Name=jacket_side_01_l, bone2Name=jacket_base, swing1LimitDegrees=80, swing2LimitDegrees=60, twistLimitDegrees=8}
 {bone1Name=jacket_side_02_l, bone2Name=jacket_side_01_l, swing1LimitDegrees=45, swing2LimitDegrees=45, twistLimitDegrees=1}
 {bone1Name=jacket_side_03_l, bone2Name=jacket_side_02_l, swing1LimitDegrees=45, swing2LimitDegrees=45, twistLimitDegrees=1}
 {bone1Name=jacket_side_04_l, bone2Name=jacket_side_03_l, swing1LimitDegrees=45, swing2LimitDegrees=45, twistLimitDegrees=1}
 {bone1Name=jacket_frt_01_l, bone2Name=jacket_base, swing1LimitDegrees=45, swing2LimitDegrees=45, twistLimitDegrees=2}
 {bone1Name=jacket_frt_02_l, bone2Name=jacket_frt_01_l, swing1LimitDegrees=15, swing2LimitDegrees=30, twistLimitDegrees=1}
 {bone1Name=jacket_frt_03_l, bone2Name=jacket_frt_02_l, swing1LimitDegrees=15, swing2LimitDegrees=30, twistLimitDegrees=1}
 {bone1Name=jacket_frt_04_l, bone2Name=jacket_frt_03_l, swing1LimitDegrees=15, swing2LimitDegrees=30, twistLimitDegrees=1}
 {
  bone1Name=calf_l
  bone2Name=thigh_l
  swing1LimitDegrees=45
  swing1Motion=Free
  swing2LimitDegrees=45
  swing2Motion=Free
  twistLimitDegrees=45
  twistMotion=Free
 }
 {bone1Name=thigh_l, bone2Name=pelvis, swing1LimitDegrees=45, swing2LimitDegrees=45, twistLimitDegrees=45}
 {
  bone1Name=jacket_frtSide_01_r
  bone2Name=jacket_base
  swing1LimitDegrees=80
  swing1Motion=Free
  swing2LimitDegrees=60
  swing2Motion=Free
  twistLimitDegrees=8
  twistMotion=Free
 }
 {bone1Name=jacket_frtSide_02_r, bone2Name=jacket_frtSide_01_r, swing1LimitDegrees=45, swing2LimitDegrees=45, twistLimitDegrees=1}
 {bone1Name=jacket_frtSide_03_r, bone2Name=jacket_frtSide_02_r, swing1LimitDegrees=45, swing2LimitDegrees=45, twistLimitDegrees=1}
 {bone1Name=jacket_frtSide_04_r, bone2Name=jacket_frtSide_03_r, swing1LimitDegrees=45, swing2LimitDegrees=45, twistLimitDegrees=1}
 {bone1Name=jacket_frt_01_r, bone2Name=jacket_base, swing1LimitDegrees=45, swing2LimitDegrees=45, twistLimitDegrees=2}
 {bone1Name=jacket_frt_02_r, bone2Name=jacket_frt_01_r, swing1LimitDegrees=35, swing2LimitDegrees=35, twistLimitDegrees=1}
 {bone1Name=jacket_frt_03_r, bone2Name=jacket_frt_02_r, swing1LimitDegrees=45, swing2LimitDegrees=45, twistLimitDegrees=0.1}
 {bone1Name=jacket_frt_04_r, bone2Name=jacket_frt_03_r, swing1LimitDegrees=45, swing2LimitDegrees=45, twistLimitDegrees=0.1}
 {bone1Name=jacket_side_01_r, bone2Name=jacket_base, swing1LimitDegrees=80, swing2LimitDegrees=60, twistLimitDegrees=8}
 {bone1Name=jacket_frt_05_r, bone2Name=jacket_frt_04_r, swing1LimitDegrees=45, swing2LimitDegrees=45, twistLimitDegrees=0.1}
 {bone1Name=jacket_frt_05_l, bone2Name=jacket_frt_04_l, swing1LimitDegrees=15, swing2LimitDegrees=30, twistLimitDegrees=1}
 {bone1Name=jacket_frtSide_01_l, bone2Name=jacket_base, swing1LimitDegrees=45, swing2LimitDegrees=45, twistLimitDegrees=45}
 {bone1Name=jacket_frtSide_02_l, bone2Name=jacket_frtSide_01_l, swing1LimitDegrees=10, swing2LimitDegrees=30, twistLimitDegrees=2}
 {bone1Name=jacket_frtSide_03_l, bone2Name=jacket_frtSide_02_l, swing1LimitDegrees=10, swing2LimitDegrees=30, twistLimitDegrees=2}
 {bone1Name=jacket_frtSide_04_l, bone2Name=jacket_frtSide_03_l, swing1LimitDegrees=10, swing2LimitDegrees=30, twistLimitDegrees=2}
 {bone1Name=jacket_side_02_r, bone2Name=jacket_side_01_r, swing1LimitDegrees=15, swing2LimitDegrees=30, twistLimitDegrees=0.1}
 {bone1Name=jacket_side_03_r, bone2Name=jacket_side_02_r, swing1LimitDegrees=15, swing2LimitDegrees=30, twistLimitDegrees=0.1}
 {bone1Name=jacket_side_04_r, bone2Name=jacket_side_03_r, swing1LimitDegrees=15, swing2LimitDegrees=30, twistLimitDegrees=0.1}
 {
  bone1Name=thigh_twist_01_l
  bone2Name=thigh_l
  swing1LimitDegrees=45
  swing1Motion=Free
  swing2LimitDegrees=45
  swing2Motion=Free
  twistLimitDegrees=45
  twistMotion=Free
 }
 {
  bone1Name=pelvis
  bone2Name=root
  swing1LimitDegrees=45
  swing1Motion=Free
  swing2LimitDegrees=45
  swing2Motion=Free
  twistLimitDegrees=45
  twistMotion=Free
 }
]
