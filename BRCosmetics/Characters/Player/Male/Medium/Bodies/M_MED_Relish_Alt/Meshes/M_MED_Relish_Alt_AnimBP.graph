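# Animation Blueprint for the M_MED_Relish_Alt player body that drives jacket and collar secondary motion.
# Cached source-pose chains layer bone constraints and additive corrective offsets around RigidBody simulation, with separate standing and crouch physics assets.
# Run, raised-thigh and raised-arm alphas weight correctives, many interpolated, that shift jacket panels with the limbs.
# Bool blends select crouching or standing chains, or the untouched source pose when rigid bodies are disabled.
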AnimBlueprint /BRCosmetics/Characters/Player/Male/Medium/Bodies/M_MED_Relish_Alt/Meshes/M_MED_Relish_Alt_AnimBP
target_skeleton=&/M_MED_Relish_Alt_Skeleton
graphs [
 AnimGraph {nodes [
   Blend Poses by bool(BlendPose_0=Component To Local(), bActiveValue=Get bIsInFrontEnd()) {blendType=Linear}
   Component To Local()
   n1 = "Constraint - jacket_frt_01_l (PPPP)"!(
    ComponentPose
    Alpha=1
    ConstraintWeights_0=0.5515
    ConstraintWeights_1:float=0.1493
    ConstraintWeights_2:float=0.2291
    ConstraintWeights_3:float=0.0701
   ) -> (Pose) {
    alphaBoolBlend {blendOption=Linear}
    boneToModify {boneName=jacket_frt_01_l}
    constraintSetup [
     {targetBone {boneName=jacket_frt_01_l}}
     {targetBone {boneName=dyn_lapel_l}}
     {targetBone {boneName=spine_04}}
     {targetBone {boneName=jacket_frtSide_01_l}}
    ]
    constraintWeights [1,1,1,1]
   }
   n2 = "Constraint - jacket_frt_02_l (P)"!(ComponentPose=@n1, Alpha=1, ConstraintWeights_0=1, ConstraintWeights_1:float=0) -> (Pose) {
    alphaBoolBlend {blendOption=Linear}
    boneToModify {boneName=jacket_frt_02_l}
    constraintSetup [{targetBone {boneName=jacket_frt_02_l}}]
    constraintWeights [1,0]
   }
   n3 = "Constraint - jacket_frt_03_l (PPPPPPP)"!(
    ComponentPose=@n2
    Alpha=1
    ConstraintWeights_0=0.0012
    ConstraintWeights_1:float=0.3358
    ConstraintWeights_2:float=0.1168
    ConstraintWeights_3:float=0.0545
    ConstraintWeights_4:float=0.0004
    ConstraintWeights_5:float=0.4622
    ConstraintWeights_6:float=0.0292
   ) -> (Pose) {
    alphaBoolBlend {blendOption=Linear}
    boneToModify {boneName=jacket_frt_03_l}
    constraintSetup [
     {targetBone {boneName=thigh_twist_01_r}}
     {targetBone {boneName=spine_01}}
     {targetBone {boneName=spine_02}}
     {targetBone {boneName=spine_03}}
     {targetBone {boneName=spine_04}}
     {targetBone {boneName=pelvis}}
     {targetBone {boneName=thigh_twist_01_l}}
    ]
    constraintWeights [1,1,1,1,1,1,1]
   }
   n4 = "Constraint - jacket_frt_04_l (PPP)"!(ComponentPose=@n3, Alpha=1, ConstraintWeights_0=0.0393, ConstraintWeights_1:float=0.5661, ConstraintWeights_2:float=0.3946) -> (Pose) {
    alphaBoolBlend {blendOption=Linear}
    boneToModify {boneName=jacket_frt_04_l}
    constraintSetup [{targetBone {boneName=spine_01}}, {targetBone {boneName=thigh_twist_01_l}}, {targetBone {boneName=pelvis}}]
    constraintWeights [1,1,1]
   }
   "Constraint - jacket_frt_05_l (PPP)"!(ComponentPose=@n4, Alpha=1, ConstraintWeights_0=0.0055, ConstraintWeights_1:float=0.988, ConstraintWeights_2:float=0.0065) -> (Pose) {
    alphaBoolBlend {blendOption=Linear}
    boneToModify {boneName=jacket_frt_05_l}
    constraintSetup [{targetBone {boneName=pelvis}}, {targetBone {boneName=thigh_twist_01_l}}, {targetBone {boneName=thigh_l}}]
    constraintWeights [1,1,1]
   }
   n5 = "Constraint - jacket_frt_04_l (PP)"!(ComponentPose, Alpha=1, ConstraintWeights_0=0.6, ConstraintWeights_1:float=0.4) -> (Pose) {
    alphaBoolBlend {blendOption=Linear}
    boneToModify {boneName=jacket_frt_04_l}
    constraintSetup [{targetBone {boneName=thigh_l}}, {targetBone {boneName=jacket_frtSide_05_l}}]
    constraintWeights [1,1]
   }
   n6 = "Constraint - jacket_side_03_l (PP)"!(ComponentPose=@n5, Alpha=1, ConstraintWeights_0=0.8, ConstraintWeights_1:float=0.2) -> (Pose) {
    alphaBoolBlend {blendOption=Linear}
    boneToModify {boneName=jacket_side_03_l}
    constraintSetup [{targetBone {boneName=thigh_l}}, {targetBone {boneName=pelvis}}]
    constraintWeights [1,1]
   }
   "Constraint - jacket_frtSide_03_l (PP)"!(ComponentPose=@n6, Alpha=1, ConstraintWeights_0=0.8, ConstraintWeights_1:float=0.2) -> (Pose) {
    alphaBoolBlend {blendOption=Linear}
    boneToModify {boneName=jacket_frtSide_03_l}
    constraintSetup [{targetBone {boneName=thigh_l}}, {targetBone {boneName=pelvis}}]
    constraintWeights [1,1]
   }
   n7 = "Constraint - jacket_frt_04_r (PP)"!(ComponentPose, Alpha=1, ConstraintWeights_0=0.6, ConstraintWeights_1:float=0.4) -> (Pose) {
    alphaBoolBlend {blendOption=Linear}
    alphaScaleBiasClamp {bClampResult=true, clampMax=0.8}
    boneToModify {boneName=jacket_frt_04_r}
    constraintSetup [{targetBone {boneName=thigh_r}}, {targetBone {boneName=spine_01}}]
    constraintWeights [1,1]
   }
   n8 = "Constraint - jacket_side_03_r (PP)"!(ComponentPose=@n7, Alpha=1, ConstraintWeights_0=0.8, ConstraintWeights_1:float=0.2) -> (Pose) {
    alphaBoolBlend {blendOption=Linear}
    alphaScaleBiasClamp {bClampResult=true, clampMax=0.8}
    boneToModify {boneName=jacket_side_03_r}
    constraintSetup [{targetBone {boneName=thigh_r}}, {targetBone {boneName=pelvis}}]
    constraintWeights [1,1]
   }
   "Constraint - jacket_frtSide_03_r (PP)"!(ComponentPose=@n8, Alpha=1, ConstraintWeights_0=0.8, ConstraintWeights_1:float=0.2) -> (Pose) {
    alphaBoolBlend {blendOption=Linear}
    alphaScaleBiasClamp {bClampResult=true, clampMax=0.8}
    boneToModify {boneName=jacket_frtSide_03_r}
    constraintSetup [{targetBone {boneName=thigh_r}}, {targetBone {boneName=pelvis}}]
    constraintWeights [1,1]
   }
   n9 = "Constraint - jacket_frt_04_l (PP)"(Alpha=1, ConstraintWeights_0=0.6, ConstraintWeights_1=0.4) {
    alphaBoolBlend {blendOption=Linear}
    alphaScaleBiasClamp {bClampResult=true, clampMax=0.25}
    boneToModify {boneName=jacket_frt_04_l}
    constraintSetup [{targetBone {boneName=thigh_l}}, {targetBone {boneName=thigh_twist_01_l}}]
    constraintWeights [1,1]
   }
   "Constraint - jacket_frtSide_04_l (PP)"!(ComponentPose=@n9, Alpha=1, ConstraintWeights_0=0.6, ConstraintWeights_1:float=0.4) -> (Pose) {
    alpha=0
    alphaBoolBlend {blendOption=Linear}
    alphaScaleBiasClamp {bClampResult=true, clampMax=0.25}
    boneToModify {boneName=jacket_frtSide_04_l}
    constraintSetup [{targetBone {boneName=thigh_l}}, {targetBone {boneName=thigh_twist_01_l}}]
    constraintWeights [1,1]
   }
   n10 = "Constraint - jacket_frt_04_r (PP)"(Alpha=1, ConstraintWeights_0=0.6, ConstraintWeights_1=0.4) {
    alpha=0
    alphaBoolBlend {blendOption=Linear}
    alphaScaleBiasClamp {bClampResult=true, clampMax=0.25}
    boneToModify {boneName=jacket_frt_04_r}
    constraintSetup [{targetBone {boneName=thigh_r}}, {targetBone {boneName=thigh_twist_01_r}}]
    constraintWeights [1,1]
   }
   "Constraint - jacket_side_04_r (PP)"!(ComponentPose=@n10, Alpha=1, ConstraintWeights_0=0.6, ConstraintWeights_1:float=0.4) -> (Pose) {
    alphaBoolBlend {blendOption=Linear}
    alphaScaleBiasClamp {bClampResult=true, clampMax=0.25}
    boneToModify {boneName=jacket_side_04_r}
    constraintSetup [{targetBone {boneName=thigh_r}}, {targetBone {boneName=pelvis}}]
    constraintWeights [1,1]
   }
   n11 = "Constraint - jacket_frt_02_l (P)"!(ComponentPose, Alpha=0.3, ConstraintWeights_0=1) -> (Pose) {
    alphaBoolBlend {blendOption=Linear}
    boneToModify {boneName=jacket_frt_02_l}
    constraintSetup [{targetBone {boneName=thigh_l}}]
    constraintWeights [1]
   }
   n12 = "Constraint - jacket_frt_04_l (P)"(ComponentPose=@n11, Alpha=0.3, ConstraintWeights_0=1) {
    alphaBoolBlend {blendOption=Linear}
    boneToModify {boneName=jacket_frt_04_l}
    constraintSetup [{targetBone {boneName=thigh_l}}]
    constraintWeights [1]
   }
   "Constraint - jacket_frt_04_r (P)"(ComponentPose=@n12, Alpha=0.3, ConstraintWeights_0=1) {
    alphaBoolBlend {blendOption=Linear}
    boneToModify {boneName=jacket_frt_04_r}
    constraintSetup [{targetBone {boneName=thigh_r}}]
    constraintWeights [1]
   }
   "Transform (Modify) Bone - Bone: jacket_frt_04_r"!(ComponentPose, Alpha=1, Translation="0,0,0", Rotation="2,-25,4") -> (Pose) {
    rotationMode=BMM_Additive
    rotationSpace=BCS_BoneSpace
    translationMode=BMM_Additive
    translationSpace=BCS_BoneSpace
    alphaBoolBlend {blendOption=Linear}
    alphaScaleBiasClamp {bClampResult=true}
    boneToModify {boneName=jacket_frt_04_r}
    rotation=<rot 2 -25 4>
   }
   n13 = "Transform (Modify) Bone - Bone: jacket_frt_04_r"(Alpha=1, Translation="0.000,-0.000,-0.873", Rotation="0,0,0") {
    rotationMode=BMM_Additive
    rotationSpace=BCS_BoneSpace
    translationMode=BMM_Additive
    translationSpace=BCS_BoneSpace
    alphaBoolBlend {blendOption=Linear}
    boneToModify {boneName=jacket_frt_04_r}
    translation {z=-0.873}
   }
   n14 = "Transform (Modify) Bone - Bone: jacket_frtSide_03_r"!(ComponentPose=@n13, Alpha=1, Translation="-0.000,1.500,0.000", Rotation="0,0,0") -> (Pose) {
    rotationMode=BMM_Additive
    rotationSpace=BCS_BoneSpace
    translationMode=BMM_Additive
    translationSpace=BCS_BoneSpace
    alphaBoolBlend {blendOption=Linear}
    boneToModify {boneName=jacket_frtSide_03_r}
    translation {y=1.5}
   }
   "Transform (Modify) Bone - Bone: jacket_side_01_r"!(ComponentPose=@n14, Alpha=1, Translation="-0.000,1.177,4.478", Rotation="0.000,-5.000,0.000") -> (Pose) {
    rotationMode=BMM_Additive
    rotationSpace=BCS_BoneSpace
    translationMode=BMM_Additive
    translationSpace=BCS_BoneSpace
    alphaBoolBlend {blendOption=Linear}
    boneToModify {boneName=jacket_side_01_r}
    rotation {yaw=-5}
    translation {y=1.177, z=4.478}
   }
   "Transform (Modify) Bone - Bone: jacket_frtSide_01_l"!(ComponentPose, Alpha=1, Translation="0,0,0", Rotation="0,0,0") -> (Pose) {
    alpha=0
    rotationMode=BMM_Replace
    rotationSpace=BCS_BoneSpace
    translationMode=BMM_Replace
    translationSpace=BCS_BoneSpace
    alphaBoolBlend {blendOption=Linear}
    boneToModify {boneName=jacket_frtSide_01_l}
   }
   n15 = "Constraint - jacket_frt_01_r (PPPP)"!(
    ComponentPose=Local To Component(LocalPose=Use cached pose 'Source Pose'())
    Alpha=1
    ConstraintWeights_0=0.1032
    ConstraintWeights_1:float=0.2683
    ConstraintWeights_2:float=0.1418
    ConstraintWeights_3:float=0.4866
   ) -> (Pose) {
    alphaBoolBlend {blendOption=Linear}
    boneToModify {boneName=jacket_frt_01_r}
    constraintSetup [
     {targetBone {boneName=dyn_lapel_r}}
     {targetBone {boneName=spine_04}}
     {targetBone {boneName=jacket_frtSide_01_r}}
     {targetBone {boneName=jacket_frt_01_r}}
    ]
    constraintWeights [1,1,1,1]
   }
   n16 = "Constraint - jacket_frt_02_r (P)"!(ComponentPose=@n15, Alpha=1, ConstraintWeights_0=1, ConstraintWeights_1:float=0) -> (Pose) {
    alphaBoolBlend {blendOption=Linear}
    boneToModify {boneName=jacket_frt_02_r}
    constraintSetup [{targetBone {boneName=jacket_frt_02_r}}]
    constraintWeights [1,0]
   }
   n17 = "Constraint - jacket_frt_03_r (PPPP)"!(
    ComponentPose=@n16
    Alpha=1
    ConstraintWeights_0=0.4068
    ConstraintWeights_1:float=0.0992
    ConstraintWeights_2:float=0.0393
    ConstraintWeights_3:float=0.4546
   ) -> (Pose) {
    alphaBoolBlend {blendOption=Linear}
    boneToModify {boneName=jacket_frt_03_r}
    constraintSetup [{targetBone {boneName=spine_01}}, {targetBone {boneName=spine_02}}, {targetBone {boneName=spine_03}}, {targetBone {boneName=pelvis}}]
    constraintWeights [1,1,1,1]
   }
   n18 = "Constraint - jacket_frt_04_r (PPP)"!(ComponentPose=@n17, Alpha=1, ConstraintWeights_0=0.0393, ConstraintWeights_1:float=0.5661, ConstraintWeights_2:float=0.3946) -> (Pose) {
    alphaBoolBlend {blendOption=Linear}
    boneToModify {boneName=jacket_frt_04_r}
    constraintSetup [{targetBone {boneName=spine_01}}, {targetBone {boneName=thigh_twist_01_r}}, {targetBone {boneName=pelvis}}]
    constraintWeights [1,1,1]
   }
   n19 = "Transform (Modify) Bone - Bone: jacket_frt_01_l"!(ComponentPose=@n18, Alpha=1, Translation="-1.000,10,-3.000", Rotation="-29.235,-24.482,-21.541") -> (Pose) {
    alpha=0
    rotationMode=BMM_Additive
    rotationSpace=BCS_BoneSpace
    translationMode=BMM_Additive
    translationSpace=BCS_BoneSpace
    alphaBoolBlend {blendOption=Linear}
    alphaScaleBiasClamp {bClampResult=true, bInterpResult=true, clampMax=0.5, interpSpeedDecreasing=20, interpSpeedIncreasing=20}
    boneToModify {boneName=jacket_frt_01_l}
    rotation=<rot -29.235 -24.4818 -21.5409>
    translation=<-1 10 -3>
   }
   n20 = "Transform (Modify) Bone - Bone: jacket_frt_03_l"!(ComponentPose=@n19, Alpha=1, Translation="0,0,0", Rotation="5,0.000,0.000") -> (Pose) {
    alpha=0
    rotationMode=BMM_Replace
    rotationSpace=BCS_BoneSpace
    translationMode=BMM_Replace
    translationSpace=BCS_BoneSpace
    alphaBoolBlend {blendOption=Linear}
    boneToModify {boneName=jacket_frt_03_l}
    rotation {pitch=5}
   }
   n21 = "Transform (Modify) Bone - Bone: jacket_frt_01_r"!(ComponentPose=@n20, Alpha=1, Translation="-7.328,2.702,0.820", Rotation="18.658,-9.916,-74.330") -> (Pose) {
    alpha=0
    rotationMode=BMM_Additive
    rotationSpace=BCS_BoneSpace
    translationMode=BMM_Additive
    translationSpace=BCS_BoneSpace
    alphaBoolBlend {blendOption=Linear}
    alphaScaleBiasClamp {bClampResult=true, bInterpResult=true, clampMax=0.5, interpSpeedDecreasing=20, interpSpeedIncreasing=20}
    boneToModify {boneName=jacket_frt_01_r}
    rotation=<rot 18.658 -9.9161 -74.3297>
    translation=<-7.328 2.702 0.82>
   }
   n22 = "Transform (Modify) Bone - Bone: jacket_frt_03_r"!(ComponentPose=@n21, Alpha=1, Translation="1.822,8.135,-11.343", Rotation="14.984,0.197,0.699") -> (Pose) {
    alpha=0
    rotationMode=BMM_Replace
    rotationSpace=BCS_BoneSpace
    translationMode=BMM_Additive
    translationSpace=BCS_BoneSpace
    alphaBoolBlend {blendOption=Linear}
    boneToModify {boneName=jacket_frt_03_r}
    rotation=<rot 14.984 0.197 0.699>
    translation=<1.822 8.135 -11.343>
   }
   n23 = "Constraint - jacket_frt_04_l (PP)"(ComponentPose=@n22, Alpha=1, ConstraintWeights_0=0.8, ConstraintWeights_1=0.2) {
    alpha=0
    alphaBoolBlend {blendOption=Linear}
    alphaScaleBiasClamp {bClampResult=true, clampMax=0.15}
    boneToModify {boneName=jacket_frt_04_l}
    constraintSetup [{targetBone {boneName=thigh_l}}, {targetBone {boneName=pelvis}}]
    constraintWeights [1,1]
   }
   n24 = "Constraint - jacket_frt_04_r (PP)"(ComponentPose=@n23, Alpha=1, ConstraintWeights_0=0.8, ConstraintWeights_1=0.2) {
    alphaBoolBlend {blendOption=Linear}
    alphaScaleBiasClamp {bClampResult=true, clampMax=0.15}
    boneToModify {boneName=jacket_frt_04_r}
    constraintSetup [{targetBone {boneName=thigh_r}}, {targetBone {boneName=pelvis}}]
    constraintWeights [1,1]
   }
   n25 = "Constraint - jacket_frt_03_l (PP)"!(ComponentPose=@n24, Alpha=1, ConstraintWeights_0=0.6, ConstraintWeights_1:float=0.4) -> (Pose) {
    alphaBoolBlend {blendOption=Linear}
    alphaScaleBiasClamp {bClampResult=true, clampMax=0.2}
    boneToModify {boneName=jacket_frt_03_l}
    constraintSetup [{targetBone {boneName=jacket_frtSide_04_l}}, {targetBone {boneName=jacket_side_04_r}}]
    constraintWeights [1,1]
   }
   n26 = "Constraint - jacket_frt_03_r (PP)"!(ComponentPose=@n25, Alpha=1, ConstraintWeights_0=0.6, ConstraintWeights_1:float=0.4) -> (Pose) {
    alphaBoolBlend {blendOption=Linear}
    alphaScaleBiasClamp {bClampResult=true, clampMax=0.2}
    boneToModify {boneName=jacket_frt_03_r}
    constraintSetup [{targetBone {boneName=jacket_frtSide_04_r}}, {targetBone {boneName=jacket_side_04_r}}]
    constraintWeights [1,1]
   }
   "Transform (Modify) Bone - Bone: jacket_frt_03_r"!(ComponentPose=@n26, bAlphaBoolEnabled:bool=True, Translation="0,0,1.5", Rotation="0,0,0", Scale:Vector="1,1,1") -> (Pose) {
    alphaInputType=Bool
    rotationMode=BMM_Additive
    rotationSpace=BCS_BoneSpace
    scaleMode=BMM_Additive
    scaleSpace=BCS_BoneSpace
    translationMode=BMM_Additive
    translationSpace=BCS_BoneSpace
    alphaBoolBlend {blendOption=Linear}
    alphaScaleBiasClamp {bInterpResult=true, interpSpeedIncreasing=30}
    boneToModify {boneName=jacket_frt_03_r}
    translation {z=1.5}
   }
   "Transform (Modify) Bone - Bone: jacket_frt_01_r"!(ComponentPose, bAlphaBoolEnabled:bool=True, Translation="0,0,1.75", Rotation="0,0,0", Scale:Vector="1,1,1") -> (Pose) {
    alphaInputType=Bool
    rotationMode=BMM_Additive
    rotationSpace=BCS_BoneSpace
    scaleMode=BMM_Additive
    scaleSpace=BCS_BoneSpace
    translationMode=BMM_Additive
    translationSpace=BCS_BoneSpace
    alphaBoolBlend {blendOption=Linear}
    alphaScaleBiasClamp {bInterpResult=true, interpSpeedIncreasing=30}
    boneToModify {boneName=jacket_frt_01_r}
    translation {z=1.75}
   }
   "Transform (Modify) Bone - Bone: jacket_frtSide_01_r"!(ComponentPose, Alpha=1, Translation="0,0,0", Rotation="-11.585,-14.493,1.276") -> (Pose) {
    alpha=0
    rotationMode=BMM_Additive
    rotationSpace=BCS_BoneSpace
    translationMode=BMM_Additive
    translationSpace=BCS_BoneSpace
    alphaBoolBlend {blendOption=Linear}
    alphaScaleBiasClamp {bClampResult=true, bInterpResult=true, clampMax=0.5, interpSpeedDecreasing=20, interpSpeedIncreasing=20}
    boneToModify {boneName=jacket_frtSide_01_r}
    rotation=<rot -11.585 -14.4931 1.276>
   }
   "Transform (Modify) Bone - Bone: jacket_frt_04_l"!(ComponentPose, Alpha=1, Translation="0,0,0", Rotation="-15,-30,4") -> (Pose) {
    rotationMode=BMM_Additive
    rotationSpace=BCS_BoneSpace
    translationMode=BMM_Additive
    translationSpace=BCS_BoneSpace
    alphaBoolBlend {blendOption=Linear}
    boneToModify {boneName=jacket_frt_04_l}
    rotation=<rot -15 -30 4>
   }
   RigidBody!(ComponentPose, Alpha=0.6, OverrideWorldGravity:Vector="0,0,-980", ExternalForce) -> (Pose) {
    bClampLinearTranslationLimitToRefPose=true
    bOverrideWorldGravity=true
    overridePhysicsAsset=&/M_MED_Relish_Alt_Coat_FE_Physics
    simulationSpace=BaseBoneSpace
    alphaBoolBlend {blendOption=Linear}
    baseBoneRef {boneName=root}
    componentAppliedLinearAccClamp=<3000 3000 3000>
    componentLinearAccScale=<0.2 0.2 0.4>
    componentLinearVelScale=<0.5 0.5 0.5>
    overrideWorldGravity {z=-980}
   }
   RigidBody(Alpha=0.6, OverrideWorldGravity="0,0,-980") {
    bClampLinearTranslationLimitToRefPose=true
    bOverrideWorldGravity=true
    overridePhysicsAsset=&M_MED_Relish_Alt_Coat_FE_Physics
    simulationSpace=BaseBoneSpace
    alphaBoolBlend {blendOption=Linear}
    baseBoneRef {boneName=root}
    componentAppliedLinearAccClamp=<3000 3000 3000>
    componentLinearAccScale=<0.2 0.2 0.4>
    componentLinearVelScale=<0.5 0.5 0.5>
    overrideWorldGravity {z=-980}
   }
   n27 = "Constraint - dyn_collar_bk (P)"(ComponentPose=Local To Component(LocalPose=Use cached pose 'Source Pose'()), Alpha=1, ConstraintWeights_0=1) {
    alphaBoolBlend {blendOption=Linear}
    boneToModify {boneName=dyn_collar_bk}
    constraintSetup [{targetBone {boneName=head}}]
    constraintWeights [1]
   }
   n28 = "Constraint - dyn_collal_l (P)"(ComponentPose=@n27, Alpha=0.65, ConstraintWeights_0=1) {
    alphaBoolBlend {blendOption=Linear}
    boneToModify {boneName=dyn_collal_l}
    constraintSetup [{targetBone {boneName=clavicle_l}}]
    constraintWeights [1]
   }
   n29 = "Constraint - dyn_collar_r (P)"(ComponentPose=@n28, Alpha=0.65, ConstraintWeights_0=1) {
    alphaBoolBlend {blendOption=Linear}
    boneToModify {boneName=dyn_collar_r}
    constraintSetup [{targetBone {boneName=clavicle_r}}]
    constraintWeights [1]
   }
   n30 = RigidBody(ComponentPose=@n29, OverrideWorldGravity=Get GravityOverride()) {
    bClampLinearTranslationLimitToRefPose=true
    bOverrideWorldGravity=true
    overridePhysicsAsset=&/M_MED_Relish_Alt_Coat_Physics
    simulationSpace=BaseBoneSpace
    alphaBoolBlend {blendOption=Linear}
    baseBoneRef {boneName=root}
    componentAppliedLinearAccClamp=<500 500 500>
    overrideWorldGravity {z=-980}
   }
   n31 = "Transform (Modify) Bone - Bone: jacket_frt_01_l"!(ComponentPose=@n30, Alpha=1, Translation="0,5,0", Rotation="-0.433,0.000,0.000", Scale:Vector="1,1,1") -> (Pose) {
    rotationMode=BMM_Additive
    translationMode=BMM_Additive
    translationSpace=BCS_BoneSpace
    alphaBoolBlend {blendOption=Linear}
    boneToModify {boneName=jacket_frt_01_l}
    rotation {pitch=-0.433}
    translation {y=5}
   }
   n32 = "Transform (Modify) Bone - Bone: jacket_frt_01_r"!(ComponentPose=@n31, Alpha:float=1, Translation="0,5,0", Rotation="0,0,0", Scale="1,1,1") -> (Pose) {
    rotationMode=BMM_Additive
    translationMode=BMM_Additive
    translationSpace=BCS_BoneSpace
    alphaBoolBlend {blendOption=Linear}
    boneToModify {boneName=jacket_frt_01_r}
    translation {y=5}
   }
   n33 = "Transform (Modify) Bone - Bone: jacket_side_01_l"(ComponentPose=@n32, Alpha=1, Translation="4,0,0.000", Rotation="13.810,0.000,0.000", Scale="1,1,1") {
    rotationMode=BMM_Additive
    translationMode=BMM_Additive
    translationSpace=BCS_BoneSpace
    alphaBoolBlend {blendOption=Linear}
    boneToModify {boneName=jacket_side_01_l}
    rotation {pitch=13.81}
    translation {x=4}
   }
   n34 = "Transform (Modify) Bone - Bone: jacket_side_01_r"!(ComponentPose=@n33, Alpha=1, Translation="-4,0,0.000", Rotation="-11,0.000,0.000", Scale:Vector="1,1,1") -> (Pose) {
    rotationMode=BMM_Additive
    translationMode=BMM_Additive
    translationSpace=BCS_BoneSpace
    alphaBoolBlend {blendOption=Linear}
    boneToModify {boneName=jacket_side_01_r}
    rotation {pitch=-11}
    translation {x=-4}
   }
   n35 = "Transform (Modify) Bone - Bone: jacket_back_01_l"(ComponentPose=@n34, Alpha=1, Translation="0.000,3.5,0.000", Rotation="0.000,0.000,5", Scale="1,1,1") {
    rotationMode=BMM_Additive
    translationMode=BMM_Additive
    translationSpace=BCS_BoneSpace
    alphaBoolBlend {blendOption=Linear}
    boneToModify {boneName=jacket_back_01_l}
    rotation {roll=5}
    translation {y=3.5}
   }
   n36 = "Transform (Modify) Bone - Bone: jacket_back_01_r"(ComponentPose=@n35, Alpha=1, Translation="0.000,3.5,0.000", Rotation="0.000,0.000,5", Scale="1,1,1") {
    rotationMode=BMM_Additive
    translationMode=BMM_Additive
    translationSpace=BCS_BoneSpace
    alphaBoolBlend {blendOption=Linear}
    boneToModify {boneName=jacket_back_01_r}
    rotation {roll=5}
    translation {y=3.5}
   }
   n37 = "Transform (Modify) Bone - Bone: dyn_collar_bk"(ComponentPose=@n36, Alpha=1, Translation="5,-2,0.000", Rotation="0,0,0", Scale="1,1,1") {
    rotationMode=BMM_Additive
    rotationSpace=BCS_BoneSpace
    translationMode=BMM_Additive
    translationSpace=BCS_BoneSpace
    alphaBoolBlend {blendOption=Linear}
    boneToModify {boneName=dyn_collar_bk}
    translation=<5 -2>
   }
   n38 = Component To Local(ComponentPose=@n37)
   n39 = Get FrontEnd Alpha() -> (FrontEnd Alpha:double)
   n40 = "Constraint - jacket_frt_02_l (P)"(ComponentPose=Local To Component(LocalPose=Use cached pose 'Source Pose'()), Alpha=0.4, ConstraintWeights_0=1) {
    alpha=0
    alphaBoolBlend {blendOption=Linear}
    boneToModify {boneName=jacket_frt_02_l}
    constraintSetup [{targetBone {boneName=thigh_l}}]
    constraintWeights [1]
   }
   n41 = "Constraint - jacket_frt_03_l (P)"(ComponentPose=@n40, Alpha=0.4, ConstraintWeights_0=1) {
    alpha=0
    alphaBoolBlend {blendOption=Linear}
    boneToModify {boneName=jacket_frt_03_l}
    constraintSetup [{targetBone {boneName=thigh_l}}]
    constraintWeights [1]
   }
   n42 = "Constraint - jacket_frt_04_l (P)"(ComponentPose=@n41, Alpha=0.4, ConstraintWeights_0=1) {
    alpha=0
    alphaBoolBlend {blendOption=Linear}
    boneToModify {boneName=jacket_frt_04_l}
    constraintSetup [{targetBone {boneName=thigh_l}}]
    constraintWeights [1]
   }
   n43 = "Transform (Modify) Bone - Bone: jacket_frt_01_l"!(ComponentPose=@n42, Alpha=1, Translation="0,0,0", Rotation="10,-25,20") -> (Pose) {
    alpha=0
    rotationMode=BMM_Additive
    rotationSpace=BCS_BoneSpace
    translationMode=BMM_Additive
    translationSpace=BCS_BoneSpace
    alphaBoolBlend {blendOption=Linear}
    alphaScaleBiasClamp {bInterpResult=true, interpSpeedDecreasing=20, interpSpeedIncreasing=20}
    boneToModify {boneName=jacket_frt_01_l}
    rotation=<rot 10 -25 20>
   }
   n44 = "Transform (Modify) Bone - Bone: jacket_frtSide_01_l"(ComponentPose=@n43, Alpha=1, Translation="0,0,0", Rotation="-15,-30,4") {
    alpha=0
    rotationMode=BMM_Additive
    rotationSpace=BCS_BoneSpace
    translationMode=BMM_Additive
    translationSpace=BCS_BoneSpace
    alphaBoolBlend {blendOption=Linear}
    boneToModify {boneName=jacket_frtSide_01_l}
    rotation=<rot -15 -30 4>
   }
   n45 = "Constraint - jacket_frt_02_r (P)"!(ComponentPose=@n44, Alpha=0.4, ConstraintWeights_0=1) -> (Pose) {
    alpha=0
    alphaBoolBlend {blendOption=Linear}
    boneToModify {boneName=jacket_frt_02_r}
    constraintSetup [{targetBone {boneName=thigh_r}}]
    constraintWeights [1]
   }
   n46 = "Constraint - jacket_frt_03_r (P)"(ComponentPose=@n45, Alpha=0.4, ConstraintWeights_0=1) {
    alpha=0
    alphaBoolBlend {blendOption=Linear}
    boneToModify {boneName=jacket_frt_03_r}
    constraintSetup [{targetBone {boneName=thigh_r}}]
    constraintWeights [1]
   }
   n47 = "Constraint - jacket_frt_04_r (P)"(ComponentPose=@n46, Alpha=0.4, ConstraintWeights_0=1) {
    alpha=0
    alphaBoolBlend {blendOption=Linear}
    boneToModify {boneName=jacket_frt_04_r}
    constraintSetup [{targetBone {boneName=thigh_r}}]
    constraintWeights [1]
   }
   n48 = "Transform (Modify) Bone - Bone: jacket_frt_01_r"!(ComponentPose=@n47, Alpha=1, Translation="0,0,0", Rotation="12,-30,-5") -> (Pose) {
    alpha=0
    rotationMode=BMM_Additive
    rotationSpace=BCS_BoneSpace
    translationMode=BMM_Additive
    translationSpace=BCS_BoneSpace
    alphaBoolBlend {blendOption=Linear}
    alphaScaleBiasClamp {bClampResult=true}
    boneToModify {boneName=jacket_frt_01_r}
    rotation=<rot 12 -30 -5>
   }
   n49 = "Transform (Modify) Bone - Bone: jacket_frtSide_01_r"(ComponentPose=@n48, Alpha=1, Translation="0.211,0.790,0.000", Rotation="0.000,-15.000,0.000") {
    alpha=0
    rotationMode=BMM_Additive
    rotationSpace=BCS_BoneSpace
    translationMode=BMM_Additive
    translationSpace=BCS_BoneSpace
    alphaBoolBlend {blendOption=Linear}
    alphaScaleBiasClamp {bClampResult=true}
    boneToModify {boneName=jacket_frtSide_01_r}
    rotation {yaw=-15.0001}
    translation=<0.211 0.79>
   }
   n50 = "Constraint - jacket_frt_05_l (PP)"!(ComponentPose=@n49, Alpha=1, ConstraintWeights_0=0.5, ConstraintWeights_1:float=0.5) -> (Pose) {
    alphaBoolBlend {blendOption=Linear}
    boneToModify {boneName=jacket_frt_05_l}
    constraintSetup [{targetBone {boneName=jacket_frtSide_05_l}}, {targetBone {boneName=jacket_side_05_l}}]
    constraintWeights [1,1]
   }
   n51 = "Constraint - jacket_frt_04_r (PP)"(ComponentPose=@n50, Alpha=1, ConstraintWeights_0=0.5, ConstraintWeights_1=0.5) {
    alphaBoolBlend {blendOption=Linear}
    boneToModify {boneName=jacket_frt_04_r}
    constraintSetup [{targetBone {boneName=jacket_frtSide_04_r}}, {targetBone {boneName=jacket_side_04_r}}]
    constraintWeights [1,1]
   }
   n52 = RigidBody(ComponentPose=@n51, OverrideWorldGravity="0,0,-980") {
    bClampLinearTranslationLimitToRefPose=true
    bOverrideWorldGravity=true
    overridePhysicsAsset=&M_MED_Relish_Alt_Coat_Physics
    simulationSpace=BaseBoneSpace
    alphaBoolBlend {blendOption=Linear}
    baseBoneRef {boneName=spine_01}
    componentAppliedLinearAccClamp=<5000 5000 5000>
    componentLinearAccScale=<0.2 0.2 0.4>
    componentLinearVelScale=<0.5 0.5 0.5>
    overrideWorldGravity {z=-980}
   }
   n53 = Get run_alpha() -> (run_alpha:double)
   n54 = "Constraint - jacket_frt_04_l (R)"(ComponentPose=@n52, Alpha=@n53, ConstraintWeights_0=1) {
    alphaBoolBlend {blendOption=Linear}
    alphaScaleBiasClamp {bClampResult=true, bInterpResult=true, clampMax=0.65, interpSpeedDecreasing=15, interpSpeedIncreasing=15}
    boneToModify {boneName=jacket_frt_04_l}
    constraintSetup [{transformType=Rotation, targetBone {boneName=thigh_l}}]
    constraintWeights [1]
   }
   n55 = "Constraint - jacket_frt_04_r (R)"(ComponentPose=@n54, Alpha=@n53, ConstraintWeights_0=1) {
    alphaBoolBlend {blendOption=Linear}
    alphaScaleBiasClamp {bClampResult=true, bInterpResult=true, clampMax=0.5, interpSpeedDecreasing=15, interpSpeedIncreasing=15}
    boneToModify {boneName=jacket_frt_04_r}
    constraintSetup [{transformType=Rotation, targetBone {boneName=thigh_r}}]
    constraintWeights [1]
   }
   n56 = "Transform (Modify) Bone - Bone: jacket_frt_01_l"(ComponentPose=@n55, Alpha=@n53, Translation="0,0,0", Rotation="0,-15,0") {
    rotationMode=BMM_Additive
    rotationSpace=BCS_BoneSpace
    translationMode=BMM_Additive
    translationSpace=BCS_BoneSpace
    alphaBoolBlend {blendOption=Linear}
    alphaScaleBiasClamp {bInterpResult=true, interpSpeedDecreasing=25, interpSpeedIncreasing=25}
    boneToModify {boneName=jacket_frt_01_l}
    rotation {yaw=-15}
   }
   n57 = "Transform (Modify) Bone - Bone: jacket_frt_01_r"(ComponentPose=@n56, Alpha=@n53, Translation="0,0,0", Rotation="0.000,-5,0.000") {
    rotationMode=BMM_Additive
    rotationSpace=BCS_BoneSpace
    translationMode=BMM_Additive
    translationSpace=BCS_BoneSpace
    alphaBoolBlend {blendOption=Linear}
    alphaScaleBiasClamp {bInterpResult=true, interpSpeedDecreasing=25, interpSpeedIncreasing=25}
    boneToModify {boneName=jacket_frt_01_r}
    rotation {yaw=-5}
   }
   n58 = "Constraint - jacket_frt_04_l (R)"(ComponentPose=@n57, Alpha=@n53, ConstraintWeights_0=1) {
    alphaBoolBlend {blendOption=Linear}
    alphaScaleBiasClamp {bInterpResult=true, interpSpeedDecreasing=15, interpSpeedIncreasing=15}
    boneToModify {boneName=jacket_frt_04_l}
    constraintSetup [{transformType=Rotation, targetBone {boneName=thigh_l}}]
    constraintWeights [1]
   }
   n59 = "Constraint - jacket_frt_04_r (R)"(ComponentPose=@n58, Alpha=@n53, ConstraintWeights_0=1) {
    alphaBoolBlend {blendOption=Linear}
    alphaScaleBiasClamp {bInterpResult=true, interpSpeedDecreasing=15, interpSpeedIncreasing=15}
    boneToModify {boneName=jacket_frt_04_r}
    constraintSetup [{transformType=Rotation, targetBone {boneName=thigh_r}}]
    constraintWeights [1]
   }
   n60 = "Transform (Modify) Bone - Bone: jacket_frt_02_r"!(ComponentPose=@n59, Alpha=1, Translation="3,2.5,-3.5") -> (Pose) {
    alpha=0
    rotationMode=BMM_Additive
    rotationSpace=BCS_BoneSpace
    scaleSpace=BCS_BoneSpace
    translationMode=BMM_Additive
    translationSpace=BCS_BoneSpace
    alphaBoolBlend {blendOption=Linear}
    alphaScaleBiasClamp {bInterpResult=true, interpSpeedDecreasing=60, interpSpeedIncreasing=30}
    boneToModify {boneName=jacket_frt_02_r}
    translation=<3 2.5 -3.5>
   }
   n61 = Get thigh_rot_up_r() -> (thigh_rot_up_r:double)
   n62 = "Transform (Modify) Bone - Bone: jacket_frt_02_r"!(ComponentPose=@n60, Alpha=@n61, Translation="0,5,0", Rotation:Rotator="-25,10,0") -> (Pose) {
    rotationMode=BMM_Additive
    rotationSpace=BCS_BoneSpace
    scaleSpace=BCS_BoneSpace
    translationMode=BMM_Additive
    translationSpace=BCS_BoneSpace
    alphaBoolBlend {blendOption=Linear}
    alphaScaleBiasClamp {bInterpResult=true, interpSpeedDecreasing=60, interpSpeedIncreasing=30}
    boneToModify {boneName=jacket_frt_02_r}
    rotation {pitch=-25, yaw=10}
    translation {y=5}
   }
   n63 = "Transform (Modify) Bone - Bone: jacket_frt_01_r"(ComponentPose=@n62, Alpha=@n61, Translation="-1.5,1.5,-1.4", Rotation="0.000,-10,0.000") {
    rotationMode=BMM_Additive
    rotationSpace=BCS_BoneSpace
    translationMode=BMM_Additive
    translationSpace=BCS_BoneSpace
    alphaBoolBlend {blendOption=Linear}
    alphaScaleBiasClamp {bInterpResult=true, interpSpeedDecreasing=60, interpSpeedIncreasing=60}
    boneToModify {boneName=jacket_frt_01_r}
    rotation {yaw=-10}
    translation=<-1.5 1.5 -1.4>
   }
   n64 = "Transform (Modify) Bone - Bone: jacket_frt_03_r"!(ComponentPose=@n63, Alpha:float=1, Translation="0,0.000,-2", Rotation="0,0,0") -> (Pose) {
    alpha=0
    rotationMode=BMM_Additive
    rotationSpace=BCS_BoneSpace
    translationMode=BMM_Additive
    translationSpace=BCS_BoneSpace
    alphaBoolBlend {blendOption=Linear}
    alphaScaleBiasClamp {bInterpResult=true, interpSpeedDecreasing=60, interpSpeedIncreasing=30}
    boneToModify {boneName=jacket_frt_03_r}
    translation {z=-2}
   }
   n65 = "Transform (Modify) Bone - Bone: jacket_frtSide_03_r"(ComponentPose=@n64, Alpha=1, Translation="0,0,0", Rotation="-2,0,0") {
    alpha=0
    rotationMode=BMM_Additive
    rotationSpace=BCS_BoneSpace
    translationMode=BMM_Additive
    translationSpace=BCS_BoneSpace
    alphaBoolBlend {blendOption=Linear}
    alphaScaleBiasClamp {bInterpResult=true, interpSpeedDecreasing=60, interpSpeedIncreasing=30}
    boneToModify {boneName=jacket_frtSide_03_r}
    rotation {pitch=-2}
   }
   n66 = "Transform (Modify) Bone - Bone: jacket_frtSide_02_r"!(ComponentPose=@n65, Alpha=@n61, Translation="0,4,0", Rotation="0.000,-25,0.000") -> (Pose) {
    rotationMode=BMM_Additive
    rotationSpace=BCS_BoneSpace
    translationMode=BMM_Additive
    translationSpace=BCS_BoneSpace
    alphaBoolBlend {blendOption=Linear}
    alphaScaleBiasClamp {bInterpResult=true, interpSpeedDecreasing=60, interpSpeedIncreasing=30}
    boneToModify {boneName=jacket_frtSide_02_r}
    rotation {yaw=-25}
    translation {y=4}
   }
   n67 = "Transform (Modify) Bone - Bone: jacket_frt_01_l"(ComponentPose=@n66, Alpha=Get thigh_rot_up_l() -> (thigh_rot_up_l:double), Translation="-0.5,0.800,0.5", Rotation="0,0,0") {
    rotationMode=BMM_Additive
    rotationSpace=BCS_BoneSpace
    translationMode=BMM_Additive
    translationSpace=BCS_BoneSpace
    alphaBoolBlend {blendOption=Linear}
    alphaScaleBiasClamp {bInterpResult=true, interpSpeedDecreasing=30, interpSpeedIncreasing=30}
    boneToModify {boneName=jacket_frt_01_l}
    translation=<-0.5 0.8 0.5>
   }
   n68 = "Transform (Modify) Bone - Bone: jacket_frt_01_l"!(ComponentPose=Local To Component(LocalPose=Use cached pose 'Source Pose'()), Alpha=1, Translation="0,5,0") -> (Pose) {
    rotationMode=BMM_Additive
    rotationSpace=BCS_BoneSpace
    translationMode=BMM_Additive
    translationSpace=BCS_BoneSpace
    alphaBoolBlend {blendOption=Linear}
    boneToModify {boneName=jacket_frt_01_l}
    translation {y=5}
   }
   n69 = "Transform (Modify) Bone - Bone: jacket_frt_01_r"(ComponentPose=@n68, Alpha=1, Translation="2.028,-1.261,1.678", Rotation="21.504,-70,-8.937") {
    rotationMode=BMM_Additive
    rotationSpace=BCS_BoneSpace
    translationMode=BMM_Additive
    translationSpace=BCS_BoneSpace
    alphaBoolBlend {blendOption=Linear}
    alphaScaleBiasClamp {bInterpResult=true, interpSpeedDecreasing=5, interpSpeedIncreasing=5}
    boneToModify {boneName=jacket_frt_01_r}
    rotation=<rot 21.504 -70 -8.937>
    translation=<2.0281 -1.2611 1.6781>
   }
   n70 = "Transform (Modify) Bone - Bone: jacket_frt_01_l"!(ComponentPose=@n69, Alpha=1, Rotation:Rotator="0.000,-10,0.000") -> (Pose) {
    rotationMode=BMM_Additive
    rotationSpace=BCS_BoneSpace
    translationMode=BMM_Additive
    translationSpace=BCS_BoneSpace
    alphaBoolBlend {blendOption=Linear}
    alphaScaleBiasClamp {bInterpResult=true, interpSpeedDecreasing=5, interpSpeedIncreasing=5}
    boneToModify {boneName=jacket_frt_01_l}
    rotation {yaw=-10}
   }
   n71 = "Constraint - jacket_frt_03_r (P)"(ComponentPose=@n70, Alpha=0.2, ConstraintWeights_0=1) {
    alphaBoolBlend {blendOption=Linear}
    boneToModify {boneName=jacket_frt_03_r}
    constraintSetup [{targetBone {boneName=thigh_r}}]
    constraintWeights [1]
   }
   n72 = "Constraint - jacket_frt_04_r (P)"(ComponentPose=@n71, Alpha=1, ConstraintWeights_0=1) {
    alphaBoolBlend {blendOption=Linear}
    alphaScaleBiasClamp {bInterpResult=true}
    boneToModify {boneName=jacket_frt_04_r}
    constraintSetup [{targetBone {boneName=thigh_twist_01_r}}]
    constraintWeights [1]
   }
   n73 = "Constraint - jacket_frt_05_r (PP)"!(ComponentPose=@n72, Alpha=0.5, ConstraintWeights_0=0.4, ConstraintWeights_1:float=0.6) -> (Pose) {
    alphaBoolBlend {blendOption=Linear}
    boneToModify {boneName=jacket_frt_05_r}
    constraintSetup [{targetBone {boneName=thigh_r}}, {targetBone {boneName=dyn_zipper_r_MSTR}}]
    constraintWeights [1,1]
   }
   n74 = "Constraint - jacket_frt_03_l (PP)"(ComponentPose=@n73, Alpha=0.5, ConstraintWeights_0=0.8, ConstraintWeights_1=0.2) {
    alphaBoolBlend {blendOption=Linear}
    boneToModify {boneName=jacket_frt_03_l}
    constraintSetup [{targetBone {boneName=thigh_l}}, {targetBone {boneName=dyn_zipper_l_MSTR}}]
    constraintWeights [1,1]
   }
   n75 = Get thigh_rot_up_l()
   n76 = "Transform (Modify) Bone - Bone: jacket_frt_01_l"!(ComponentPose=@n74, Alpha=@n75, Translation:Vector="-2.712,0.800,1.2", Rotation="-2.616,-81.623,72.148") -> (Pose) {
    rotationMode=BMM_Additive
    rotationSpace=BCS_BoneSpace
    translationMode=BMM_Additive
    translationSpace=BCS_BoneSpace
    alphaBoolBlend {blendOption=Linear}
    boneToModify {boneName=jacket_frt_01_l}
    rotation=<rot -2.616 -81.6229 72.1479>
    translation=<-2.712 0.8 1.2>
   }
   n77 = "Transform (Modify) Bone - Bone: jacket_frtSide_01_l"(ComponentPose=@n76, Alpha=@n75, Translation="-5.317,-3.790,-5.543", Rotation="4.428,-22.137,-49.071") {
    rotationMode=BMM_Additive
    rotationSpace=BCS_BoneSpace
    translationMode=BMM_Additive
    translationSpace=BCS_BoneSpace
    alphaBoolBlend {blendOption=Linear}
    boneToModify {boneName=jacket_frtSide_01_l}
    rotation=<rot 4.428 -22.1367 -49.0709>
    translation=<-5.317 -3.79 -5.543>
   }
   n78 = "Transform (Modify) Bone - Bone: jacket_side_01_l"!(ComponentPose=@n77, Alpha=@n75, Translation="2.858,4.232,-2.847", Rotation="-10.000,0.000,0.000") -> (Pose) {
    rotationMode=BMM_Additive
    rotationSpace=BCS_BoneSpace
    translationMode=BMM_Additive
    translationSpace=BCS_BoneSpace
    alphaBoolBlend {blendOption=Linear}
    boneToModify {boneName=jacket_side_01_l}
    rotation {pitch=-10}
    translation=<2.858 4.232 -2.847>
   }
   n79 = Get thigh_rot_up_r()
   n80 = "Transform (Modify) Bone - Bone: jacket_frt_01_r"(ComponentPose=@n78, Alpha=@n79, Translation="1.450,0.000,0.000", Rotation="0.000,30,0.000") {
    rotationMode=BMM_Additive
    rotationSpace=BCS_BoneSpace
    translationMode=BMM_Additive
    translationSpace=BCS_BoneSpace
    alphaBoolBlend {blendOption=Linear}
    boneToModify {boneName=jacket_frt_01_r}
    rotation {yaw=30}
    translation {x=1.45}
   }
   n81 = "Transform (Modify) Bone - Bone: jacket_side_01_r"!(ComponentPose=@n80, Alpha=@n79, Rotation="-15,0.000,0.000") -> (Pose) {
    rotationMode=BMM_Additive
    rotationSpace=BCS_BoneSpace
    translationMode=BMM_Additive
    translationSpace=BCS_BoneSpace
    alphaBoolBlend {blendOption=Linear}
    boneToModify {boneName=jacket_side_01_r}
    rotation {pitch=-15}
   }
   n82 = "Constraint - jacket_frt_02_l (P)"(ComponentPose=@n81, Alpha=1, ConstraintWeights_0=1) {
    alphaBoolBlend {blendOption=Linear}
    boneToModify {boneName=jacket_frt_02_l}
    constraintSetup [{targetBone {boneName=thigh_twist_01_l}}]
    constraintWeights [1]
   }
   n83 = RigidBody(ComponentPose=@n82, OverrideWorldGravity="0,0,-980") {
    bClampLinearTranslationLimitToRefPose=true
    bOverrideWorldGravity=true
    overridePhysicsAsset=&/M_MED_Relish_Alt_Crouch_Physics
    simulationSpace=BaseBoneSpace
    alphaBoolBlend {blendOption=Linear}
    baseBoneRef {boneName=pelvis}
    componentAppliedLinearAccClamp=<5000 5000 5000>
    overrideWorldGravity {z=-980}
   }
   n84 = "Transform (Modify) Bone - Bone: jacket_frt_01_l"(ComponentPose=@n83, Alpha=Get arm_up_l() -> (arm_up_l:double), Translation="-2.500,-2,-4.5", Rotation="-4,-10,-15") {
    rotationMode=BMM_Additive
    rotationSpace=BCS_BoneSpace
    translationMode=BMM_Additive
    translationSpace=BCS_BoneSpace
    alphaBoolBlend {blendOption=Linear}
    alphaScaleBiasClamp {bInterpResult=true}
    boneToModify {boneName=jacket_frt_01_l}
    rotation=<rot -4 -10 -15>
    translation=<-2.5 -2 -4.5>
   }
   n85 = Blend Poses by bool(
    BlendPose_0=Blend Poses by bool(
     BlendPose_0=Component To Local(ComponentPose=@n84)
     BlendPose_1=Component To Local(ComponentPose=@n67)
     BlendTime_1=0.5
     bActiveValue=Get bIsCrouching()
    ) {blendType=Linear}
    BlendPose_1=Use cached pose 'Source Pose'()
    bActiveValue=Get bRigidBodyEnabled()
   ) {blendType=Linear}
   n86 = RigidBody(ComponentPose=Local To Component(LocalPose=Use cached pose 'cache_PerfMode'()), OverrideWorldGravity="0,0,-980") {
    bClampLinearTranslationLimitToRefPose=true
    bOverrideWorldGravity=true
    overridePhysicsAsset=&/M_MED_Relish_Alt_Coat_FrontEnd_Physics
    simulationSpace=BaseBoneSpace
    alphaBoolBlend {blendOption=Linear, blendOutTime=3}
    baseBoneRef {boneName=spine_01}
    componentAppliedLinearAccClamp=<5000 5000 5000>
    componentLinearAccScale=<0.2 0.2 0.4>
    componentLinearVelScale=<0.5 0.5 0.5>
    overrideWorldGravity {z=-980}
   }
   n87 = "Transform (Modify) Bone - Bone: jacket_frt_01_r_base"!(ComponentPose=@n86, bAlphaBoolEnabled:bool=True, Translation="0,0,0", Rotation="0,10,0", Scale="1,1,1") -> (Pose) {
    alphaInputType=Bool
    rotationMode=BMM_Additive
    rotationSpace=BCS_BoneSpace
    alphaBoolBlend {blendInTime=3, blendOption=Linear, blendOutTime=3}
    boneToModify {boneName=jacket_frt_01_r_base}
    rotation {yaw=10}
   }
   n88 = "Transform (Modify) Bone - Bone: jacket_frt_01_l_base"!(ComponentPose=@n87, bAlphaBoolEnabled:bool=True, Translation="0,0,0", Rotation="0,-10,0", Scale="1,1,1") -> (Pose) {
    alphaInputType=Bool
    rotationMode=BMM_Additive
    rotationSpace=BCS_BoneSpace
    alphaBoolBlend {blendInTime=3, blendOption=Linear, blendOutTime=3}
    boneToModify {boneName=jacket_frt_01_l_base}
    rotation {yaw=-10}
   }
   n89 = Component To Local(ComponentPose=@n88)
   n90 = Get bIsInFrontEnd()
   n91 = "Constraint - dyn_collar_bk (P)"(
    ComponentPose=Local To Component(LocalPose=Blend Poses by bool(BlendPose_0=@n89, BlendPose_1=@n85, bActiveValue=@n90) {blendType=Linear})
    Alpha=0.8
    ConstraintWeights_0=1
   ) {
    alphaBoolBlend {blendOption=Linear}
    boneToModify {boneName=dyn_collar_bk}
    constraintSetup [{targetBone {boneName=head}}]
    constraintWeights [1]
   }
   n92 = "Constraint - dyn_collal_l (P)"(ComponentPose=@n91, Alpha=0.65, ConstraintWeights_0=1) {
    alphaBoolBlend {blendOption=Linear}
    boneToModify {boneName=dyn_collal_l}
    constraintSetup [{targetBone {boneName=clavicle_l}}]
    constraintWeights [1]
   }
   n93 = "Constraint - dyn_collar_r (P)"(ComponentPose=@n92, Alpha=0.65, ConstraintWeights_0=1) {
    alphaBoolBlend {blendOption=Linear}
    boneToModify {boneName=dyn_collar_r}
    constraintSetup [{targetBone {boneName=clavicle_r}}]
    constraintWeights [1]
   }
   n94 = "Constraint - dyn_collal_l (P)"(ComponentPose=@n93, Alpha=1, ConstraintWeights_0=1) {
    alphaBoolBlend {blendOption=Linear}
    boneToModify {boneName=dyn_collal_l}
    constraintSetup [{targetBone {boneName=clavicle_l}}]
    constraintWeights [1]
   }
   n95 = "Constraint - dyn_collar_r (P)"(ComponentPose=@n94, Alpha=1, ConstraintWeights_0=1) {
    alphaBoolBlend {blendOption=Linear}
    boneToModify {boneName=dyn_collar_r}
    constraintSetup [{targetBone {boneName=clavicle_r}}]
    constraintWeights [1]
   }
   n96 = Get Head Up() -> (Head Up:double)
   n97 = "Transform (Modify) Bone - Bone: dyn_collar_bk"!(ComponentPose=@n95, Alpha=@n96, Translation="0.000,-2,-1.5", Rotation="0,-15,0.000") -> (Pose) {
    rotationMode=BMM_Additive
    rotationSpace=BCS_BoneSpace
    translationMode=BMM_Additive
    translationSpace=BCS_BoneSpace
    alphaBoolBlend {blendOption=Linear}
    boneToModify {boneName=dyn_collar_bk}
    rotation {yaw=-15}
    translation {y=-2, z=-1.5}
   }
   n98 = "Transform (Modify) Bone - Bone: dyn_collar_r"!(ComponentPose=@n97, Alpha=@n96, Translation="0,-1.5,0", Rotation="20,1.288,-5") -> (Pose) {
    rotationMode=BMM_Additive
    rotationSpace=BCS_BoneSpace
    scaleSpace=BCS_BoneSpace
    translationMode=BMM_Additive
    translationSpace=BCS_BoneSpace
    alphaBoolBlend {blendInTime=0.1, blendOption=Linear, blendOutTime=0.2}
    alphaScaleBiasClamp {bInterpResult=true, interpSpeedDecreasing=15, interpSpeedIncreasing=15}
    boneToModify {boneName=dyn_collar_r}
    rotation=<rot 20 1.288 -5>
    translation {y=-1.5}
   }
   n99 = "Transform (Modify) Bone - Bone: dyn_collar_bk"(ComponentPose=@n98, Alpha=Get Head Down() -> (Head Down:double), Translation="-0.000,0.000,-3", Rotation="0.000,0.000,-30") {
    rotationMode=BMM_Additive
    translationMode=BMM_Additive
    alphaBoolBlend {blendOption=Linear}
    boneToModify {boneName=dyn_collar_bk}
    rotation {roll=-30}
    translation {z=-3}
   }
   n100 = Get Head Left() -> (Head Left:double)
   n101 = "Transform (Modify) Bone - Bone: dyn_collar_bk"!(ComponentPose=@n99, Alpha=@n100, Rotation="0.000,0.000,5.000", Scale:Vector="1.000,1.000,1.2") -> (Pose) {
    rotationMode=BMM_Additive
    rotationSpace=BCS_BoneSpace
    scaleMode=BMM_Additive
    scaleSpace=BCS_BoneSpace
    translationSpace=BCS_BoneSpace
    alphaBoolBlend {blendOption=Linear}
    alphaScaleBiasClamp {bInterpResult=true, interpSpeedDecreasing=20, interpSpeedIncreasing=20}
    boneToModify {boneName=dyn_collar_bk}
    rotation {roll=5.0001}
    scale {z=1.2}
   }
   n102 = "Transform (Modify) Bone - Bone: dyn_collal_l"!(ComponentPose=@n101, Alpha=@n100, Rotation="-10,0.000,0.000") -> (Pose) {
    rotationMode=BMM_Additive
    rotationSpace=BCS_BoneSpace
    scaleSpace=BCS_BoneSpace
    translationSpace=BCS_BoneSpace
    alphaBoolBlend {blendOption=Linear}
    alphaScaleBiasClamp {bInterpResult=true, interpSpeedDecreasing=2, interpSpeedIncreasing=2}
    boneToModify {boneName=dyn_collal_l}
    rotation {pitch=-10}
    scale=<0 0 0>
   }
   n103 = Get Is_Swinging() -> (Is_Swinging:double)
   n104 = "Transform (Modify) Bone - Bone: dyn_collar_bk"!(ComponentPose=@n102, Alpha=@n103, Translation:Vector="3.000,-1.5,0.000", Rotation="0,0,0", Scale="1.000,1.000,1.3") -> (Pose) {
    rotationMode=BMM_Additive
    rotationSpace=BCS_BoneSpace
    scaleMode=BMM_Additive
    scaleSpace=BCS_BoneSpace
    translationMode=BMM_Additive
    translationSpace=BCS_BoneSpace
    alphaBoolBlend {blendOption=Linear}
    alphaScaleBiasClamp {bInterpResult=true, interpSpeedDecreasing=5, interpSpeedIncreasing=5}
    boneToModify {boneName=dyn_collar_bk}
    scale {z=1.3}
    translation=<3 -1.5>
   }
   n105 = "Transform (Modify) Bone - Bone: dyn_collar_r"!(ComponentPose=@n104, Alpha=@n103, Translation="-0.000,-0.000,-2", Rotation="0,0,0", Scale:Vector="1,1,0.8") -> (Pose) {
    rotationMode=BMM_Additive
    rotationSpace=BCS_BoneSpace
    scaleMode=BMM_Additive
    scaleSpace=BCS_BoneSpace
    translationMode=BMM_Additive
    translationSpace=BCS_BoneSpace
    alphaBoolBlend {blendOption=Linear}
    alphaScaleBiasClamp {bInterpResult=true, interpSpeedDecreasing=5, interpSpeedIncreasing=5}
    boneToModify {boneName=dyn_collar_r}
    scale {z=0.8}
    translation {z=-2}
   }
   n106 = "Transform (Modify) Bone - Bone: dyn_collal_l"!(ComponentPose=@n105, Alpha=@n103, Translation:Vector="4,-2,1.5", Rotation="-5.000,0.000,10.000", Scale:Vector="0.8,1.000,0.8") -> (Pose) {
    rotationMode=BMM_Additive
    rotationSpace=BCS_BoneSpace
    scaleMode=BMM_Additive
    scaleSpace=BCS_BoneSpace
    translationMode=BMM_Additive
    translationSpace=BCS_BoneSpace
    alphaBoolBlend {blendOption=Linear}
    alphaScaleBiasClamp {bInterpResult=true, interpSpeedDecreasing=5, interpSpeedIncreasing=5}
    boneToModify {boneName=dyn_collal_l}
    rotation=<rot -5 0 10.0002>
    scale {x=0.8, z=0.8}
    translation=<4 -2 1.5>
   }
   n107 = Get Rifle() -> (Rifle:double)
   n108 = "Transform (Modify) Bone - Bone: dyn_collar_bk"!(ComponentPose=@n106, Alpha=@n107, Translation="0.000,-1.5,-1", Rotation="-3.768,-6.635,14.606") -> (Pose) {
    rotationMode=BMM_Additive
    rotationSpace=BCS_BoneSpace
    scaleMode=BMM_Additive
    scaleSpace=BCS_BoneSpace
    translationMode=BMM_Additive
    alphaBoolBlend {blendOption=Linear}
    alphaScaleBiasClamp {bInterpResult=true, interpSpeedIncreasing=15}
    boneToModify {boneName=dyn_collar_bk}
    rotation=<rot -3.768 -6.635 14.606>
    scale {z=1.25}
    translation {y=-1.5, z=-1}
   }
   n109 = "Transform (Modify) Bone - Bone: dyn_collal_l"(ComponentPose=@n108, Alpha=@n107, Translation="0,0.2,0", Rotation="0.000,0,15", Scale="1,1,1") {
    rotationMode=BMM_Additive
    rotationSpace=BCS_BoneSpace
    scaleMode=BMM_Additive
    scaleSpace=BCS_BoneSpace
    translationMode=BMM_Additive
    translationSpace=BCS_BoneSpace
    alphaBoolBlend {blendOption=Linear}
    alphaScaleBiasClamp {bInterpResult=true, interpSpeedIncreasing=30}
    boneToModify {boneName=dyn_collal_l}
    rotation {roll=15}
    translation {y=0.2}
   }
   n110 = "Transform (Modify) Bone - Bone: dyn_collar_r"(ComponentPose=@n109, Alpha=@n107, Translation="0,0,-0.5", Rotation="0,3,0", Scale="1,1,1") {
    rotationMode=BMM_Additive
    rotationSpace=BCS_BoneSpace
    scaleMode=BMM_Additive
    scaleSpace=BCS_BoneSpace
    translationMode=BMM_Additive
    translationSpace=BCS_BoneSpace
    alphaBoolBlend {blendOption=Linear}
    alphaScaleBiasClamp {bInterpResult=true, interpSpeedIncreasing=30}
    boneToModify {boneName=dyn_collar_r}
    rotation {yaw=3}
    translation {z=-0.5}
   }
   n111 = "Transform (Modify) Bone - Bone: dyn_collar_bk"!(ComponentPose=@n110, Alpha=Get Pistol() -> (Pistol:double), Translation="0.000,1,2") -> (Pose) {
    rotationMode=BMM_Additive
    rotationSpace=BCS_BoneSpace
    translationMode=BMM_Additive
    translationSpace=BCS_BoneSpace
    alphaBoolBlend {blendOption=Linear}
    alphaScaleBiasClamp {bInterpResult=true, interpSpeedIncreasing=15}
    boneToModify {boneName=dyn_collar_bk}
    translation {y=1, z=2}
   }
   n112 = "Transform (Modify) Bone - Bone: dyn_collar_r"!(ComponentPose=@n111, bAlphaBoolEnabled:bool=Get bIsTargeting(), Translation="-0.000,-0.000,-3", Scale="0.9,0.8,0.5") -> (Pose) {
    alphaInputType=Bool
    rotationMode=BMM_Additive
    rotationSpace=BCS_BoneSpace
    scaleMode=BMM_Additive
    translationMode=BMM_Additive
    translationSpace=BCS_BoneSpace
    alphaBoolBlend {blendInTime=0.1, blendOption=Linear, blendOutTime=0.2}
    boneToModify {boneName=dyn_collar_r}
    scale=<0.9 0.8 0.5>
    translation {z=-3}
   }
   n113 = Get arm_up_l()
   n114 = "Transform (Modify) Bone - Bone: dyn_collal_l"(ComponentPose=@n112, Alpha=@n113, Translation="4,0,0.000", Rotation="0,5,0", Scale="0.9,1,0.85") {
    rotationMode=BMM_Additive
    scaleMode=BMM_Additive
    translationMode=BMM_Additive
    translationSpace=BCS_BoneSpace
    alphaBoolBlend {blendOption=Linear}
    alphaScaleBiasClamp {bInterpResult=true, interpSpeedIncreasing=30}
    boneToModify {boneName=dyn_collal_l}
    rotation {yaw=5}
    scale {x=0.9, z=0.85}
    translation {x=4}
   }
   n115 = "Transform (Modify) Bone - Bone: dyn_collal_l"(ComponentPose=@n114, Alpha=@n113, Translation="0,-1.2,3", Rotation="0,0,-3", Scale="0.9,1,0.85") {
    rotationMode=BMM_Additive
    rotationSpace=BCS_BoneSpace
    scaleMode=BMM_Additive
    translationMode=BMM_Additive
    translationSpace=BCS_BoneSpace
    alphaBoolBlend {blendOption=Linear}
    alphaScaleBiasClamp {bInterpResult=true, interpSpeedIncreasing=30}
    boneToModify {boneName=dyn_collal_l}
    rotation {roll=-3}
    scale {x=0.9, z=0.85}
    translation {y=-1.2, z=3}
   }
   n116 = "Transform (Modify) Bone - Bone: dyn_collar_r"!(
    ComponentPose=@n115
    Alpha:float=Get arm_up_r() -> (arm_up_r:double)
    Translation="0,0,0"
    Rotation:Rotator="15.000,0.001,0.000"
    Scale="0.9,0.8,0.5"
   ) -> (Pose) {
    rotationMode=BMM_Additive
    rotationSpace=BCS_BoneSpace
    scaleMode=BMM_Additive
    translationMode=BMM_Additive
    translationSpace=BCS_BoneSpace
    alphaBoolBlend {blendInTime=0.1, blendOption=Linear, blendOutTime=0.2}
    alphaScaleBiasClamp {bInterpResult=true, interpSpeedIncreasing=30}
    boneToModify {boneName=dyn_collar_r}
    rotation {pitch=15, yaw=0.001}
    scale=<0.9 0.8 0.5>
   }
   n117 = RigidBody!(ComponentPose=@n116, Alpha=@n39, ExternalForce) -> (Pose) {
    overridePhysicsAsset=&/M_MED_Relish_Alt_Collar_Physics
    alphaBoolBlend {blendOption=Linear}
   }
   n118 = "Transform (Modify) Bone - Bone: dyn_lapel_l"(ComponentPose=@n117, Alpha=@n39, Translation="0,0,0", Rotation="0.000,0.000,2", Scale="1.000,1.000,1.250") {
    rotationMode=BMM_Additive
    rotationSpace=BCS_BoneSpace
    scaleMode=BMM_Additive
    scaleSpace=BCS_BoneSpace
    translationSpace=BCS_BoneSpace
    alphaBoolBlend {blendInTime=0.1, blendOption=Linear, blendOutTime=0.2}
    alphaScaleBiasClamp {interpSpeedDecreasing=15, interpSpeedIncreasing=15}
    boneToModify {boneName=dyn_lapel_l}
    rotation {roll=2}
    scale {z=1.25}
   }
   n119 = Component To Local(ComponentPose=@n118)
   n120 = Get bIsSkyDiving()
   n121 = Anim Dynamics - Bone: dyn_zipper_01_l(
    ComponentPose=Local To Component(LocalPose=Blend Poses by bool(BlendPose_0=@n38, BlendPose_1=@n119, bActiveValue=@n120) {blendType=Linear})
    Alpha=1
    ExternalForce="0,0,0"
   ) {
    angularSpringConstant=1000
    bAngularSpring=true
    bGravityOverrideInSimSpace=true
    bUseGravityOverride=true
    lODThreshold=1
    numSolverIterationsPostUpdate=4
    numSolverIterationsPreUpdate=16
    alphaBoolBlend {blendOption=Linear}
    boundBone {boneName=dyn_zipper_01_l}
    componentLinearAccScale=<0.2 0.2 0.45>
    componentLinearVelScale=<0.2 0.2 0.45>
    gravityOverride {z=-980}
    physicsBodyDefinitions [{
      sphereCollisionRadius=0
      boundBone {boneName=dyn_zipper_01_l}
      boxExtents=<8 20 20>
      constraintSetup {angularLimitsMax=<3 2 15>, angularLimitsMin=<-1 -5 -30>, angularTarget=<0 0 0>}
      localJointOffset=<5 0 0>
     }]
    sphericalLimits [{limitRadius=16, drivingBone {boneName=pelvis}, sphereLocalOffset=<-1 1 0>}]
   }
   n122 = Anim Dynamics - Bone: dyn_zipper_01_r(ComponentPose=@n121, Alpha=1, ExternalForce="0,0,0") {
    angularSpringConstant=1000
    bAngularSpring=true
    bGravityOverrideInSimSpace=true
    bUseGravityOverride=true
    lODThreshold=1
    numSolverIterationsPostUpdate=4
    numSolverIterationsPreUpdate=16
    alphaBoolBlend {blendOption=Linear}
    boundBone {boneName=dyn_zipper_01_r}
    componentLinearAccScale=<0.2 0.2 0.45>
    componentLinearVelScale=<0.2 0.2 0.45>
    gravityOverride {z=-980}
    physicsBodyDefinitions [{
      sphereCollisionRadius=0
      boundBone {boneName=dyn_zipper_01_r}
      boxExtents=<8 20 20>
      constraintSetup {angularLimitsMax=<3 2 15>, angularLimitsMin=<-1 -5 -30>, angularTarget=<0 0 0>}
      localJointOffset=<5 0 0>
     }]
    sphericalLimits [{limitRadius=16, drivingBone {boneName=pelvis}, sphereLocalOffset=<-1 1 0>}]
   }
   Output Pose(Result=Component To Local(ComponentPose=@n122)) {layerGroup=None}
   n123 = "Constraint - jacket_frt_01_l (PPPP)"(
    ComponentPose=Local To Component(LocalPose=Use cached pose 'Source Pose'())
    Alpha=1
    ConstraintWeights_0=0.5515
    ConstraintWeights_1=0.1493
    ConstraintWeights_2=0.2291
    ConstraintWeights_3=0.0701
   ) {
    alphaBoolBlend {blendOption=Linear}
    boneToModify {boneName=jacket_frt_01_l}
    constraintSetup [
     {targetBone {boneName=jacket_frt_01_l}}
     {targetBone {boneName=dyn_lapel_l}}
     {targetBone {boneName=spine_04}}
     {targetBone {boneName=jacket_frtSide_01_l}}
    ]
    constraintWeights [1,1,1,1]
   }
   n124 = "Constraint - jacket_frt_02_l (P)"!(ComponentPose=@n123, Alpha=1, ConstraintWeights_0=1, ConstraintWeights_1:float=0) -> (Pose) {
    alphaBoolBlend {blendOption=Linear}
    boneToModify {boneName=jacket_frt_02_l}
    constraintSetup [{targetBone {boneName=jacket_frt_02_l}}]
    constraintWeights [1,0]
   }
   n125 = "Constraint - jacket_frt_03_l (PPPPPPP)"(
    ComponentPose=@n124
    Alpha=1
    ConstraintWeights_0=0.0012
    ConstraintWeights_1=0.3358
    ConstraintWeights_2=0.1168
    ConstraintWeights_3=0.0545
    ConstraintWeights_4=0.0004
    ConstraintWeights_5=0.4622
    ConstraintWeights_6=0.0292
   ) {
    alphaBoolBlend {blendOption=Linear}
    boneToModify {boneName=jacket_frt_03_l}
    constraintSetup [
     {targetBone {boneName=thigh_twist_01_r}}
     {targetBone {boneName=spine_01}}
     {targetBone {boneName=spine_02}}
     {targetBone {boneName=spine_03}}
     {targetBone {boneName=spine_04}}
     {targetBone {boneName=pelvis}}
     {targetBone {boneName=thigh_twist_01_l}}
    ]
    constraintWeights [1,1,1,1,1,1,1]
   }
   n126 = "Constraint - jacket_frt_04_l (PPP)"(ComponentPose=@n125, Alpha=1, ConstraintWeights_0=0.0393, ConstraintWeights_1=0.5661, ConstraintWeights_2=0.3946) {
    alphaBoolBlend {blendOption=Linear}
    boneToModify {boneName=jacket_frt_04_l}
    constraintSetup [{targetBone {boneName=spine_01}}, {targetBone {boneName=thigh_twist_01_l}}, {targetBone {boneName=pelvis}}]
    constraintWeights [1,1,1]
   }
   n127 = "Constraint - jacket_frt_05_l (PPP)"(ComponentPose=@n126, Alpha=1, ConstraintWeights_0=0.0055, ConstraintWeights_1=0.988, ConstraintWeights_2=0.0065) {
    alphaBoolBlend {blendOption=Linear}
    boneToModify {boneName=jacket_frt_05_l}
    constraintSetup [{targetBone {boneName=pelvis}}, {targetBone {boneName=thigh_twist_01_l}}, {targetBone {boneName=thigh_l}}]
    constraintWeights [1,1,1]
   }
   n128 = "Constraint - jacket_frtSide_03_l (PPP)"!(ComponentPose=@n127, Alpha=1, ConstraintWeights_0=0.0307, ConstraintWeights_1:float=0.5817, ConstraintWeights_2:float=0.3876) -> (Pose) {
    alphaBoolBlend {blendOption=Linear}
    boneToModify {boneName=jacket_frtSide_03_l}
    constraintSetup [{targetBone {boneName=spine_01}}, {targetBone {boneName=thigh_twist_01_l}}, {targetBone {boneName=pelvis}}]
    constraintWeights [1,1,1]
   }
   n129 = "Constraint - jacket_frtSide_04_l (PPP)"!(ComponentPose=@n128, Alpha=1, ConstraintWeights_0=0.033, ConstraintWeights_1:float=0.9307, ConstraintWeights_2:float=0.0363) -> (Pose) {
    alphaBoolBlend {blendOption=Linear}
    boneToModify {boneName=jacket_frtSide_04_l}
    constraintSetup [{targetBone {boneName=pelvis}}, {targetBone {boneName=thigh_twist_01_l}}, {targetBone {boneName=thigh_l}}]
    constraintWeights [1,1,1]
   }
   n130 = "Constraint - jacket_side_03_l (PPP)"!(ComponentPose=@n129, Alpha=1, ConstraintWeights_0=0.0109, ConstraintWeights_1:float=0.3609, ConstraintWeights_2:float=0.6282) -> (Pose) {
    alphaBoolBlend {blendOption=Linear}
    boneToModify {boneName=jacket_side_03_l}
    constraintSetup [{targetBone {boneName=spine_01}}, {targetBone {boneName=thigh_twist_01_l}}, {targetBone {boneName=pelvis}}]
    constraintWeights [1,1,1]
   }
   n131 = "Constraint - jacket_side_04_l (PPP)"!(ComponentPose=@n130, Alpha=1, ConstraintWeights_0=0.0878, ConstraintWeights_1:float=0.835, ConstraintWeights_2:float=0.0772) -> (Pose) {
    alphaBoolBlend {blendOption=Linear}
    boneToModify {boneName=jacket_side_04_l}
    constraintSetup [{targetBone {boneName=pelvis}}, {targetBone {boneName=thigh_twist_01_l}}, {targetBone {boneName=thigh_l}}]
    constraintWeights [1,1,1]
   }
   n132 = "Constraint - jacket_back_03_l (PP)"!(ComponentPose=@n131, Alpha=1, ConstraintWeights_0=0.7609, ConstraintWeights_1:float=0.2391) -> (Pose) {
    alphaBoolBlend {blendOption=Linear}
    boneToModify {boneName=jacket_back_03_l}
    constraintSetup [{targetBone {boneName=pelvis}}, {targetBone {boneName=thigh_twist_01_l}}]
    constraintWeights [1,1]
   }
   n133 = "Constraint - jacket_frt_01_r (PPPP)"(ComponentPose=@n132, Alpha=1, ConstraintWeights_0=0.1032, ConstraintWeights_1=0.2683, ConstraintWeights_2=0.1418, ConstraintWeights_3=0.4866) {
    alphaBoolBlend {blendOption=Linear}
    boneToModify {boneName=jacket_frt_01_r}
    constraintSetup [
     {targetBone {boneName=dyn_lapel_r}}
     {targetBone {boneName=spine_04}}
     {targetBone {boneName=jacket_frtSide_01_r}}
     {targetBone {boneName=jacket_frt_01_r}}
    ]
    constraintWeights [1,1,1,1]
   }
   n134 = "Constraint - jacket_frt_02_r (P)"!(ComponentPose=@n133, Alpha=1, ConstraintWeights_0=1, ConstraintWeights_1:float=0) -> (Pose) {
    alphaBoolBlend {blendOption=Linear}
    boneToModify {boneName=jacket_frt_02_r}
    constraintSetup [{targetBone {boneName=jacket_frt_02_r}}]
    constraintWeights [1,0]
   }
   n135 = "Constraint - jacket_frt_03_r (PPPP)"(ComponentPose=@n134, Alpha=1, ConstraintWeights_0=0.4068, ConstraintWeights_1=0.0992, ConstraintWeights_2=0.0393, ConstraintWeights_3=0.4546) {
    alphaBoolBlend {blendOption=Linear}
    boneToModify {boneName=jacket_frt_03_r}
    constraintSetup [{targetBone {boneName=spine_01}}, {targetBone {boneName=spine_02}}, {targetBone {boneName=spine_03}}, {targetBone {boneName=pelvis}}]
    constraintWeights [1,1,1,1]
   }
   n136 = "Constraint - jacket_frt_04_r (PPP)"(ComponentPose=@n135, Alpha=1, ConstraintWeights_0=0.0393, ConstraintWeights_1=0.5661, ConstraintWeights_2=0.3946) {
    alphaBoolBlend {blendOption=Linear}
    boneToModify {boneName=jacket_frt_04_r}
    constraintSetup [{targetBone {boneName=spine_01}}, {targetBone {boneName=thigh_twist_01_r}}, {targetBone {boneName=pelvis}}]
    constraintWeights [1,1,1]
   }
   n137 = "Constraint - jacket_frtSide_03_r (PPP)"!(ComponentPose=@n136, Alpha=1, ConstraintWeights_0=0.0258, ConstraintWeights_1:float=0.6861, ConstraintWeights_2:float=0.2881) -> (Pose) {
    alphaBoolBlend {blendOption=Linear}
    boneToModify {boneName=jacket_frtSide_03_r}
    constraintSetup [{targetBone {boneName=spine_01}}, {targetBone {boneName=thigh_twist_01_r}}, {targetBone {boneName=pelvis}}]
    constraintWeights [1,1,1]
   }
   n138 = "Constraint - jacket_side_03_r (PPP)"!(ComponentPose=@n137, Alpha=1, ConstraintWeights_0=0.0108, ConstraintWeights_1:float=0.5074, ConstraintWeights_2:float=0.4818) -> (Pose) {
    alphaBoolBlend {blendOption=Linear}
    boneToModify {boneName=jacket_side_03_r}
    constraintSetup [{targetBone {boneName=spine_01}}, {targetBone {boneName=thigh_twist_01_r}}, {targetBone {boneName=pelvis}}]
    constraintWeights [1,1,1]
   }
   n139 = "Constraint - jacket_frt_04_r (PP)"(ComponentPose=@n138, Alpha=1, ConstraintWeights_0=0.014, ConstraintWeights_1=0.986) {
    alphaBoolBlend {blendOption=Linear}
    boneToModify {boneName=jacket_frt_04_r}
    constraintSetup [{targetBone {boneName=thigh_r}}, {targetBone {boneName=thigh_twist_01_r}}]
    constraintWeights [1,1]
   }
   n140 = "Constraint - jacket_frtSide_04_r (PPP)"!(ComponentPose=@n139, Alpha=1, ConstraintWeights_0=0.0261, ConstraintWeights_1:float=0.0512, ConstraintWeights_2:float=0.9227) -> (Pose) {
    alphaBoolBlend {blendOption=Linear}
    boneToModify {boneName=jacket_frtSide_04_r}
    constraintSetup [{targetBone {boneName=thigh_r}}, {targetBone {boneName=pelvis}}, {targetBone {boneName=thigh_twist_01_r}}]
    constraintWeights [1,1,1]
   }
   n141 = "Constraint - jacket_back_03_r (PP)"!(ComponentPose=@n140, Alpha=1, ConstraintWeights_0=0.7685, ConstraintWeights_1:float=0.2315) -> (Pose) {
    alphaBoolBlend {blendOption=Linear}
    boneToModify {boneName=jacket_back_03_r}
    constraintSetup [{targetBone {boneName=pelvis}}, {targetBone {boneName=thigh_twist_01_r}}]
    constraintWeights [1,1]
   }
   n142 = "Transform (Modify) Bone - Bone: jacket_frt_04_l"(ComponentPose=@n141, Alpha=1, Translation="3.063,-1.640,1.353", Rotation="0.000,0.000,-10.000") {
    rotationMode=BMM_Additive
    rotationSpace=BCS_BoneSpace
    translationMode=BMM_Additive
    translationSpace=BCS_BoneSpace
    alphaBoolBlend {blendOption=Linear}
    boneToModify {boneName=jacket_frt_04_l}
    rotation {roll=-10.0003}
    translation=<3.063 -1.6395 1.353>
   }
   n143 = "Transform (Modify) Bone - Bone: jacket_frtSide_03_l"!(ComponentPose=@n142, Alpha=1, Translation="-3.044,-6.790,3.065", Rotation="5.000,-5.000,-0.000") -> (Pose) {
    rotationMode=BMM_Additive
    rotationSpace=BCS_BoneSpace
    translationMode=BMM_Additive
    translationSpace=BCS_BoneSpace
    alphaBoolBlend {blendOption=Linear}
    boneToModify {boneName=jacket_frtSide_03_l}
    rotation {pitch=5, yaw=-5}
    translation=<-3.044 -6.79 3.065>
   }
   n144 = "Transform (Modify) Bone - Bone: jacket_side_01_l"(ComponentPose=@n143, Alpha=1, Translation="1.033,1.761,-0.090", Rotation="-5.000,0.000,0.000") {
    rotationMode=BMM_Additive
    rotationSpace=BCS_BoneSpace
    translationMode=BMM_Additive
    translationSpace=BCS_BoneSpace
    alphaBoolBlend {blendOption=Linear}
    boneToModify {boneName=jacket_side_01_l}
    rotation {pitch=-5}
    translation=<1.033 1.761 -0.09>
   }
   n145 = "Transform (Modify) Bone - Bone: jacket_frt_04_r"(ComponentPose=@n144, Alpha=1, Translation="0.000,-0.000,-0.873", Rotation="0,0,0") {
    rotationMode=BMM_Additive
    rotationSpace=BCS_BoneSpace
    translationMode=BMM_Additive
    translationSpace=BCS_BoneSpace
    alphaBoolBlend {blendOption=Linear}
    boneToModify {boneName=jacket_frt_04_r}
    translation {z=-0.873}
   }
   n146 = "Transform (Modify) Bone - Bone: jacket_frtSide_03_r"(ComponentPose=@n145, Alpha=1, Translation="-0.000,1.500,0.000", Rotation="0,0,0") {
    rotationMode=BMM_Additive
    rotationSpace=BCS_BoneSpace
    translationMode=BMM_Additive
    translationSpace=BCS_BoneSpace
    alphaBoolBlend {blendOption=Linear}
    boneToModify {boneName=jacket_frtSide_03_r}
    translation {y=1.5}
   }
   n147 = "Transform (Modify) Bone - Bone: jacket_side_01_r"!(ComponentPose=@n146, Alpha=1, Translation:Vector="-0.000,1.177,4.478", Rotation="0.000,-5.000,0.000") -> (Pose) {
    rotationMode=BMM_Additive
    rotationSpace=BCS_BoneSpace
    translationMode=BMM_Additive
    translationSpace=BCS_BoneSpace
    alphaBoolBlend {blendOption=Linear}
    boneToModify {boneName=jacket_side_01_r}
    rotation {yaw=-5}
    translation {y=1.177, z=4.478}
   }
   Save cached pose 'cache_PerfMode'(Pose=Component To Local(ComponentPose=@n147)) {cacheName=cache_PerfMode}
   n148 = Control Rig!(
    Source=Copy Pose From Mesh() {bUseAttachedParent=true}
    Alpha
    isSliding:bool=Get bIsSliding()
    isSkydiving_float:double=Get isSkydiving() -> (isSkydiving:double)
   ) -> (Pose) {
    alphaBoolBlend {blendOption=Linear}
    controlRigAssetReference {blueprintRigClass=&/M_MED_Relish_Alt_DeformRig.M_MED_Relish_Alt_DeformRig_C}
   }
   Save cached pose 'Source Pose'(Pose=@n148) {cacheName=Source Pose}
   Use cached pose 'cache_PerfMode'()
   Use cached pose 'cache_PerfMode'()
   Get run_alpha()
  ]}
 EventGraph {nodes [
   Event BlueprintUpdateAnimation()
   > n1 = Sequence()
   n2 = Sequence!(execute=[@n24.then, @n1.then_0]) -> (then_0, then_1, then_2:Exec, then_3:Exec, then_4:Exec, then_5:Exec, then_6:Exec, then_7:Exec, then_8:Exec, then_9:Exec)
   > Branch(Condition=OR Boolean(A=Get bIsSkyDiving(), B=Get bIsInFrontEnd()))
   > Set frontEndorSkydiving(frontEndorSkydiving:bool=true) -> (Output_Get:bool)
   n3 = Branch(execute=@n2.then_1, Condition=OR Boolean(A=Get bIsJumping(), B=Get bIsFalling()))
   > Set JumpingFalling(JumpingFalling:bool=true) -> (Output_Get:bool)
   Set JumpingFalling(execute=@n3.else, JumpingFalling=false)
   n4 = Gravity Override State Function DEPRECATED(
    execute=@n2.then_2
    CharacterPartAnimInstance=Self-Reference()
    GameTimeInSeconds=GetGameTimeInSeconds()
    GravityOverrideParams=Get Gravity Override Params()
   )
   > Set GravityOverride(GravityOverride=@n4)
   > Set isSkydiving(isSkydiving:double="To Float (Boolean)"(InBool=Get bIsSkyDiving())) -> (Output_Get:double)
   n5 = Get PawnMovementDirectionAzimuth() -> (PawnMovementDirectionAzimuth:float)
   n6 = float - float(A=MapRangeClamped(Value=@n5, InRangeA=-45, InRangeB=-90), B=MapRangeClamped(Value=@n5, InRangeA=-90, InRangeB=-135)) {is=title K2Node_CallFunction}
   n7 = Get bIsMoving2D()
   Set Left_direct(execute=@n2.then_3, Left_direct:double=SelectFloat(A="Absolute (Float)"(A=@n6), B=0, bPickA=@n7)) -> (Output_Get:double)
   n8 = Get PawnMovementDirectionAzimuth()
   n9 = float - float(A=MapRangeClamped(Value=@n8, InRangeA=45, InRangeB=90), B=MapRangeClamped(Value=@n8, InRangeA=90, InRangeB=135))
   n10 = Get bIsMoving2D()
   > Set Right_direct(Right_direct:double=SelectFloat(A="Absolute (Float)"(A=@n9), B=0, bPickA=@n10)) -> (Output_Get:double)
   Set aim_pitch_float(execute=@n2.then_4, aim_pitch_float:double=MapRangeClamped(Value=Get AimPitch(), InRangeA=0, InRangeB=90, OutRangeA=1, OutRangeB=0.35)) -> (Output_Get:double)
   n11 = BreakTransform(InTransform=GetDeltaTransformFromRefPose(self=GetOwnerSkeletalMeshComponent(), BoneName=upperarm_r, BaseName=spine_05))
   Set arm_up_r(execute=@n2.then_5, arm_up_r:double=MapRangeClamped(Value=BreakRotator(InRot=@n11.Rotation).Yaw, InRangeA=60, InRangeB=100)) -> (Output_Get:double)
   n12 = BreakTransform(InTransform=GetDeltaTransformFromRefPose(self=GetOwnerSkeletalMeshComponent(), BoneName=upperarm_l, BaseName=spine_05))
   > Set arm_up_l(arm_up_l:double=MapRangeClamped(Value=BreakRotator(InRot=@n12.Rotation).Yaw, InRangeA=60, InRangeB=100)) -> (Output_Get:double)
   n13 = GetOwnerSkeletalMeshComponent()
   n14 = BreakRotator(InRot=BreakTransform(InTransform=GetDeltaTransformFromRefPose(self=@n13, BoneName=head, BaseName=spine_05)).Rotation)
   > Set Head Up(Head Up:double=MapRangeClamped(Value=@n14.Yaw, InRangeA=-3, InRangeB=-40)) -> (Output_Get:double)
   > Set Head Down(Head Down:double=MapRangeClamped(Value=@n14.Yaw, InRangeA=3, InRangeB=20)) -> (Output_Get:double)
   > Set Head Left(Head Left:double=MapRangeClamped(Value=@n14.Pitch, InRangeA=1, InRangeB=-1)) -> (Output_Get:double)
   > n15 = Branch(Condition="float > float"(A=Get PawnSpeed2D(), B=5) {is=title K2Node_CallFunction})
   > Set run_alpha(run_alpha:double=100) -> (Output_Get:double)
   Set run_alpha(execute=@n15.else, run_alpha=0)
   n16 = BreakTransform(InTransform=GetDeltaTransformFromRefPose(self=GetOwnerSkeletalMeshComponent(), BoneName=thigh_r, BaseName=pelvis))
   Set thigh_rot_up_r(execute=@n2.then_6, thigh_rot_up_r:double=MapRangeClamped(Value=BreakRotator(InRot=@n16.Rotation).Yaw, InRangeA=-15, InRangeB=-100)) -> (Output_Get:double)
   n17 = BreakTransform(InTransform=GetDeltaTransformFromRefPose(self=GetOwnerSkeletalMeshComponent(), BoneName=thigh_l, BaseName=pelvis))
   > Set thigh_rot_up_l(thigh_rot_up_l:double=MapRangeClamped(Value=BreakRotator(InRot=@n17.Rotation).Yaw, InRangeA=-10, InRangeB=-100)) -> (Output_Get:double)
   n18 = Cast To FortPlayerAnimInstance(execute=@n2.then_7, Object=GetAnimInstance(self=GetOwnerSkeletalMeshComponent()))
   > Set Pistol(Pistol:double=0) -> (Output_Get:double)
   > Set Rifle(Rifle:double=0) -> (Output_Get:double)
   > n19 = Switch on EFortWeaponCoreAnimation!(execute, Selection:EFortWeaponCoreAnimation Enum=Get WeaponCoreAnim(self=@n18)) -> (
    Melee:Exec
    Pistol:Exec
    Shotgun:Exec
    PaperBlueprint:Exec
    Rifle:Exec
    MeleeOneHand:Exec
    MachinePistol:Exec
    RocketLauncher:Exec
    GrenadeLauncher:Exec
    GoingCommando:Exec
    AssaultRifle:Exec
    TacticalShotgun:Exec
    SniperRifle:Exec
    TrapPlacement:Exec
    ShoulderLauncher:Exec
    AbilityDecoTool:Exec
    Crossbow:Exec
    C4:Exec
    RemoteControl:Exec
    DualWield:Exec
    AR_BullPup:Exec
    AR_ForwardGrip:Exec
    MedPackPaddles:Exec
    SMG_P90:Exec
    AR_DrumGun:Exec
    Consumable_Small:Exec
    Consumable_Large:Exec
    Balloon:Exec
    MountedTurret:Exec
    CreativeTool:Exec
    ExplosiveBow:Exec
    AshtonIndigo:Exec
    AshtonChicago:Exec
    MeleeDualWield:Exec
    Unarmed:Exec
   )
   Branch(execute=[
     @n19.AssaultRifle
     @n19.Crossbow
     @n19.GrenadeLauncher
     @n19.Rifle
     @n19.RocketLauncher
     @n19.Shotgun
     @n19.SniperRifle
     @n19.TacticalShotgun
    ])
   > Set Rifle(Rifle=1)
   > n20 = Branch(Condition=Get bIsTargeting())
   > Set isAimingRifle(isAimingRifle:bool=true) -> (Output_Get:bool)
   Set isAimingRifle(execute=@n20.else, isAimingRifle=false)
   Branch(execute=[@n19.MachinePistol, @n19.Pistol])
   > Set Pistol(Pistol=1)
   > n21 = Branch(Condition=Get bIsTargeting())
   > Set isAiming_Pistol(isAiming_Pistol:bool=true) -> (Output_Get:bool)
   Set isAiming_Pistol(execute=@n21.else, isAiming_Pistol=false)
   n22 = Branch(execute=@n2.then_8, Condition=Get bIsPlayingMeleeAnim())
   > Set Is_Swinging(Is_Swinging:double=1) -> (Output_Get:double)
   Set Is_Swinging(execute=@n22.else, Is_Swinging=0)
   n23 = Branch(execute=@n2.then_9, Condition=Get bIsInFrontEnd())
   > Set FrontEnd Alpha(FrontEnd Alpha:double=0.8) -> (Output_Get:double)
   Set FrontEnd Alpha(execute=@n23.else, FrontEnd Alpha=0)
   Branch(execute=@n1.then_1, Condition=Get bIsInFrontEnd())
   > n24 = ResetDynamics(InTeleportType=TeleportPhysics)
   MapRangeClamped(Value=0, InRangeA=40, InRangeB=120)
   MapRangeClamped(Value=0, InRangeA=50, InRangeB=105)
   TryGetPawnOwner()
   MapRangeClamped(Value=0, InRangeA=40, InRangeB=120)
   AND Boolean(A=Get bIsCrouching(), B=@n7)
   AND Boolean(A=Get bIsCrouching(), B=@n10)
  ]}
]
variables [
 aim_pitch_float {type=number}
 AimUp {type=number}
 arm_up_l {type=number}
 arm_up_r {type=number}
 FrontEnd Alpha {type=number}
 frontEndorSkydiving {type=boolean}
 Gravity Override Params {type=GravityOverrideParamsStruct}
 GravityOverride {type=Vector}
 Head Down {type=number}
 Head Left {type=number}
 Head Up {type=number}
 Head_Right {type=number}
 InComponentSpace {type=boolean}
 InComponentSpaceLastFrame {type=boolean}
 InterpSpeed {type=number}
 Is_Swinging {type=number}
 isAiming {type=boolean}
 isAiming_Pistol {type=boolean}
 isAimingRifle {type=boolean}
 isSkydiving {type=number}
 JumpingFalling {type=boolean}
 Left_direct {type=number}
 LockedOnSimSpace {type=boolean}
 PartSkelMesh {type=SkeletalMeshComponent}
 Pistol {type=number}
 Rifle {type=number}
 Right Direct {type=number}
 Right_direct {type=number}
 run_alpha {type=number}
 sim_space_rotation {type=Rotator}
 Skel_Mesh {type=SkeletalMeshComponent}
 Thigh Rot Up R {type=number}
 thigh_rot_inr_l {type=number}
 thigh_rot_inr_r {type=number}
 thigh_rot_otr_r {type=number}
 thigh_rot_up_l {type=number}
 thigh_rot_up_r {type=number}
]
